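# Niagara skydiving contrail for the Lexa cosmetic, decorating hands and feet with emissive-colored meshes, sprites, lights and ribbons.
# Mesh particles follow skeletal sockets in "Flower" arrangements with layered sine wobble, alongside socket-attached decal and glow sprites.
# bIsFrontEndPreview selects gameplay emitters, driven by a computed Dive value, or _Frontend twins, driven by ParaGlideLeanAlpha.
# Front-end ribbons trail the twins' mesh particles through particle attribute readers.
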NiagaraSystem /BRCosmetics/Athena/Effects/Particles/Skydiving/NS_Contrail_Lexa/NS_Contrail_Lexa_Frontend
dependencies {
 usedDataInterfaces [
  &NiagaraDataInterfaceCurve
  &NiagaraDataInterfaceParticleRead
  &NiagaraDataInterfaceVectorCurve
  &NiagaraDataInterfaceColorCurve
 ]
 usedDynamicInputs [&RandomBool]
 usedModules [&SystemState]
}
emitters {
 Decal {
  emitterUpdateScript [
   &EmitterState(Life Cycle Mode=::NewEnumerator1, Loop Behavior=::NewEnumerator1, Loop Duration Mode=::NewEnumerator1, UseLoopDelay=true)
   /Game/Effects/Niagara/Modules/Skydiving/FN_DivingAttributes/FN_EmitterDivingAttributes_v1(Accumulate Dive=-0.0025, Accumulate Rest=0, SkeletalMesh:NiagaraDataInterfaceSkeletalMesh=$User.SkeletalMesh)
   &SpawnBurst_Instantaneous(Spawn Count=4, Use Spawn Probability=false)
  ]
  particleSpawnScript [
   &FN_SetSpawnGroup_v2(Number of Spawn Groups=4)
   /Niagara/Modules/Spawn/Initialization/InitializeParticle(
    Color=&MultiplyLinearColor(Linear Color=$User.EmissiveColor, Scale Factor=<1 1 1 0.05>)
    Module.Write Color=true
    Module.Write Lifetime=true
    Module.Write Mass=true
    Module.Write Material Random=true
    Module.Write Position=false
    Module.Write Scale=false
    Module.Write SpriteRotation=false
    Module.Write SpriteSize=true
    Position=&SimulationPosition()
    Sprite Size=&Vector2DFromFloat(Value=40)
   )
   &DynamicMaterialParameters(
    Index 0 Param 1=2
    Index 0 Param 2=1
    Index 0 Param 3=1
    Module.Param0WriteEnabled=true
    Module.Param1WriteEnabled=true
    Module.Param2WriteEnabled=true
    Module.Param3WriteEnabled=true
   )
  ]
  particleUpdateScript [
   &ParticleState(Kill Particles When Lifetime Has Elapsed=false, Let Infinitely Lived Particles Die When Emitter Deactivates=true)
   &SkeletalMeshLocation(
    Bone / Socket Index=&SpawnGroup()
    Bone Sampling Mode=::NewEnumerator4
    Inherit Velocity=false
    Interpolate Mesh Sampling=false
    MaskSpawnGroup=false
    Sampled Position Offset Coordinate Space:ENiagaraMeshTransforms=::NewEnumerator3
    Skeletal Mesh=$User.SkeletalMesh2
   ) {module=SkeletalMeshLocation001}
   &CameraOffset(Camera Offset Amount=25)
   /NiagaraFN/Sprite/FN_SpriteAlignmentAndFacing(
    SpriteAlignment:Vector3f=&RotateVectorByQuaternion(Vector To Rotate=<1 0 0>, Quaternion=$Output.SkeletalMeshLocation001.SampledRotation)
    SpriteFacing:Vector3f=&RotateVectorByQuaternion(
     Vector To Rotate=&MaskVectorBySpawnGroup(Masked Vector=<0 1 0>, Passthrough Vector=<0 1 0>)
     Quaternion=$Output.SkeletalMeshLocation001.SampledRotation
    )
   )
   &FN_Light_v3(
    Color=$User.EmissiveColor
    Create Color Attribute=true
    Create Exponent Attribute=true
    Create Radius Attribute=true
    Exponent=2
    Radius=50
    Scale Color:NiagaraFloat=&FloatFromCurve(FloatCurve=DataInterface() {propertyValues {Curve {keys [<at 0 2>,<at 1 4>]}}}, CurveIndex=$Emitter.Dive)
   )
   &ScaleColor(
    Scale Alpha=&FloatFromCurve(FloatCurve=DataInterface() {propertyValues {Curve {keys [<at 0 1>,<at 1 2>]}}}, CurveIndex=$Emitter.Dive)
    ScaleA=true
    ScaleRGB=true
   )
  ]
  rendererClasses [&NiagaraSpriteRendererProperties, &NiagaraLightRendererProperties]
  renderers [
   {rendererClass=&NiagaraSpriteRendererProperties}
   {rendererClass=&NiagaraLightRendererProperties, rendererIndex=1}
  ]
 }
 FN_SourceGlows {
  emitterSpawnScript [&FN_PreserveSkydiveUserVariables() {module=PreserveUserVariables}]
  emitterUpdateScript [
   &EmitterState()
   &SpawnRate(SpawnRate=8, Use Spawn Probability=false)
  ]
  particleSpawnScript [
   &FN_SetSpawnGroup_v2(Number of Spawn Groups=4)
   &InitializeParticle(
    Color=&MultiplyLinearColor(Linear Color=$User.EmissiveColor, Scale Factor=<1 1 1 1>)
    Lifetime=&UniformRangedFloat(Minimum=0.5, Maximum=0.65)
    Module.Write Color=true
    Module.Write Lifetime=true
    Module.Write Mass=true
    Module.Write Material Random=true
    Module.Write Position=false
    Module.Write Scale=false
    Module.Write SpriteRotation=false
    Module.Write SpriteSize=true
    Position=&SimulationPosition()
    Sprite Size=&Vector2DFromFloat(Value=&UniformRangedFloat(Minimum=100, Maximum=100))
   )
   &CameraOffset(Camera Offset Amount=-5)
  ]
  particleUpdateScript [
   &ParticleState()
   &SkeletalMeshLocation(
    Bone / Socket Index=$Particles.SpawnGroup
    Bone Sampling Mode=::NewEnumerator4
    Inherit Velocity=false
    Interpolate Mesh Sampling=false
    MaskSpawnGroup=false
    Skeletal Mesh=$User.SkeletalMesh2
   )
   &ScaleColor(
    Scale Alpha=&FloatFromCurve(FloatCurve=DataInterface() {propertyValues {Curve {keys [
         {arriveTangent=1.1784, interpMode=RCIM_Cubic, leaveTangent=1.1784, tangentMode=RCTM_User, time=0, value=0}
         {arriveTangent=0.002, interpMode=RCIM_Cubic, leaveTangent=0.002, tangentMode=RCTM_User, time=0.5, value=0.2}
         {arriveTangent=-1.0701, interpMode=RCIM_Cubic, leaveTangent=-1.0701, tangentMode=RCTM_User, time=1.0011, value=0.002}
        ]}}})
    ScaleA=true
    ScaleRGB=true
   )
  ]
  rendererClasses [&NiagaraSpriteRendererProperties]
  renderers [{rendererClass=&NiagaraSpriteRendererProperties}]
 }
 Frontend_Ribbon_LFoot {
  emitterSpawnScript [/Game/Effects/Niagara/Modules/Attributes/FN_Source_v2(AttributeReader=DataInterface() {propertyValues {EmitterBinding {bindingMode=Other, emitterName=LFoot_Frontend}}})]
  emitterUpdateScript [
   &EmitterState(
    Delay First Loop Only=true
    Inactive Response=::NewEnumerator1
    Life Cycle Mode=::NewEnumerator1
    Loop Delay=1
    Loop Duration=1
    UseLoopDelay=true
   )
   &SpawnBurst_Instantaneous(
    Spawn Count=2
    Spawn Group=-1
    Spawn Probability=&MakeCustomFloatFromBool(Bool=$User.bIsFrontEndPreview, True Float=1, False Float=0)
    Use Spawn Probability=true
   )
   &SpawnRate(
    Spawn Probability=&FN_FrontEndFloatModifier(Gameplay=0, Front End=1)
    SpawnRate=/Game/Effects/Niagara/DynamicInputs/Attributes/FN_GetEmitterParticleCount(AttributeReader:NiagaraDataInterfaceParticleRead=$Emitter.AttributeReader, Multiplier:NiagaraFloat=4)
    Use Spawn Probability=true
   )
   &FN_EmitterDivingAttributes_v1(Accumulate Dive=-0.0025, Accumulate Rest=0, SkeletalMesh:NiagaraDataInterfaceSkeletalMesh=$User.SkeletalMesh)
  ]
  particleSpawnScript [
   /SetVariables(Particles.Locked:NiagaraBool=&SetBoolByIntComparison(
     A=&ReturnParticleIDAttributes(Particle ID Attribute=::NewEnumerator1)
     B=3
     Comparison Type=::NewEnumerator7
    )) {isSetParameters=true, module=SetVariables_#1}
   /Game/Effects/Niagara/Modules/Attributes/FN_SourceParentID(Ribbon ID from Parent ID=true)
   /Game/Effects/Niagara/Modules/Attributes/FN_SourceLocation(Position:Vector3f=&self:AttributeReader_RenderPosition(Default on Invalid001:Vector3f=<0 0 0>))
   &InitializeRibbon(
    Color=&MultiplyLinearColor(Linear Color=$User.EmissiveColor, Scale Factor=<1 1 1 0.5>)
    Module.Write Color=true
    Module.Write Lifetime=true
    Module.Write Mass=true
    Module.Write Material Random=true
    Module.Write Position=false
    Module.Write RibbonTwist=false
    Module.Write RibbonWidth=true
    Ribbon Width=10
   )
   &AddVelocity(Velocity=<0 -800 800>)
  ]
  particleUpdateScript [
   &ParticleState()
   &SolveForcesAndVelocity(Clamp Velocity=false, Limit Acceleration=false)
   /Game/Effects/Niagara/Modules/Ribbons/FN_UpdateRibbon_v5(Locked Position:Vector3f=&self:AttributeReader_RenderPosition(Default on Invalid001:Vector3f=<0 0 0>))
   /SetVariables(Particles.RibbonWidth=5) {isSetParameters=true, module=SetVariables_#2}
   /SetVariables(Particles.RibbonFacing=&RotateVectorByQuaternion(
     Vector To Rotate=<0 1 0>
     Quaternion=&self:AttributeReader_02(Default:Quat4f=<0 0 0 1>)
    )) {isSetParameters=true, module=SetVariables_#3}
   &ScaleColor(
    Scale Alpha=&FloatFromCurve(FloatCurve=DataInterface() {propertyValues {Curve {keys [<at 0 0>,<at 0.2 1>]}}}, CurveIndex=$User.ParaGlideLeanAlpha)
    ScaleA=true
    ScaleRGB=true
   )
  ]
  rendererClasses [&NiagaraRibbonRendererProperties]
  renderers [{rendererClass=&NiagaraRibbonRendererProperties}]
 }
 Frontend_Ribbon_LHand {
  emitterSpawnScript [&FN_Source_v2(AttributeReader=DataInterface() {propertyValues {EmitterBinding {bindingMode=Other, emitterName=LHand_Frontend}}})]
  emitterUpdateScript [
   &EmitterState(
    Delay First Loop Only=true
    Inactive Response=::NewEnumerator1
    Life Cycle Mode=::NewEnumerator1
    Loop Delay=1
    Loop Duration=1
    UseLoopDelay=true
   )
   &SpawnBurst_Instantaneous(
    Spawn Count=4
    Spawn Group=-1
    Spawn Probability=&MakeCustomFloatFromBool(Bool=$User.bIsFrontEndPreview, True Float=1, False Float=0)
    Use Spawn Probability=true
   )
   &SpawnRate(
    Spawn Probability=&FN_FrontEndFloatModifier(Gameplay=0, Front End=1)
    SpawnRate=&FN_GetEmitterParticleCount(AttributeReader:NiagaraDataInterfaceParticleRead=$Emitter.AttributeReader, Multiplier:NiagaraFloat=4)
    Use Spawn Probability=true
   )
   &FN_EmitterDivingAttributes_v1(Accumulate Dive=-0.0025, Accumulate Rest=0, SkeletalMesh:NiagaraDataInterfaceSkeletalMesh=$User.SkeletalMesh)
  ]
  particleSpawnScript [
   /SetVariables(Particles.Locked:NiagaraBool=&SetBoolByIntComparison(
     A=&ReturnParticleIDAttributes(Particle ID Attribute=::NewEnumerator1)
     B=3
     Comparison Type=::NewEnumerator7
    )) {isSetParameters=true, module=SetVariables_#1}
   &FN_SourceParentID(Ribbon ID from Parent ID=true)
   &FN_SourceLocation(Position:Vector3f=&self:AttributeReader_RenderPosition(Default on Invalid001:Vector3f=<0 0 0>))
   &InitializeRibbon(
    Color=&MultiplyLinearColor(Linear Color=$User.EmissiveColor, Scale Factor=<1 1 1 0.5>)
    Module.Write Color=true
    Module.Write Lifetime=true
    Module.Write Mass=true
    Module.Write Material Random=true
    Module.Write Position=false
    Module.Write RibbonTwist=false
    Module.Write RibbonWidth=true
    Ribbon Width=10
   )
   &AddVelocity(Velocity=<0 -800 800>)
  ]
  particleUpdateScript [
   &ParticleState()
   &SolveForcesAndVelocity(Clamp Velocity=false, Limit Acceleration=false)
   &FN_UpdateRibbon_v5(Locked Position:Vector3f=&self:AttributeReader_RenderPosition(Default on Invalid001:Vector3f=<0 0 0>))
   /SetVariables(Particles.RibbonWidth=5) {isSetParameters=true, module=SetVariables_#2}
   /SetVariables(Particles.RibbonFacing=&RotateVectorByQuaternion(
     Vector To Rotate=<0 1 0>
     Quaternion=&self:AttributeReader_02(Default:Quat4f=<0 0 0 1>)
    )) {isSetParameters=true, module=SetVariables_#3}
   &ScaleColor(
    Scale Alpha=&FloatFromCurve(FloatCurve=DataInterface() {propertyValues {Curve {keys [<at 0 0>,<at 0.2 1>]}}}, CurveIndex=$User.ParaGlideLeanAlpha)
    ScaleA=true
    ScaleRGB=true
   )
  ]
  rendererClasses [&NiagaraRibbonRendererProperties]
  renderers [{rendererClass=&NiagaraRibbonRendererProperties}]
 }
 Frontend_Ribbon_RFoot {
  emitterSpawnScript [&FN_Source_v2(AttributeReader=DataInterface() {propertyValues {EmitterBinding {bindingMode=Other, emitterName=RFoot_Frontend}}})]
  emitterUpdateScript [
   &EmitterState(
    Delay First Loop Only=true
    Inactive Response=::NewEnumerator1
    Life Cycle Mode=::NewEnumerator1
    Loop Delay=1
    Loop Duration=1
    UseLoopDelay=true
   )
   &SpawnBurst_Instantaneous(
    Spawn Count=2
    Spawn Group=-1
    Spawn Probability=&MakeCustomFloatFromBool(Bool=$User.bIsFrontEndPreview, True Float=1, False Float=0)
    Use Spawn Probability=true
   )
   &SpawnRate(
    Spawn Probability=&FN_FrontEndFloatModifier(Gameplay=0, Front End=1)
    SpawnRate=&FN_GetEmitterParticleCount(AttributeReader:NiagaraDataInterfaceParticleRead=$Emitter.AttributeReader, Multiplier:NiagaraFloat=4)
    Use Spawn Probability=true
   )
   &FN_EmitterDivingAttributes_v1(Accumulate Dive=-0.0025, Accumulate Rest=0, SkeletalMesh:NiagaraDataInterfaceSkeletalMesh=$User.SkeletalMesh)
  ]
  particleSpawnScript [
   /SetVariables(Particles.Locked:NiagaraBool=&SetBoolByIntComparison(
     A=&ReturnParticleIDAttributes(Particle ID Attribute=::NewEnumerator1)
     B=3
     Comparison Type=::NewEnumerator7
    )) {isSetParameters=true, module=SetVariables_#1}
   &FN_SourceParentID(Ribbon ID from Parent ID=true)
   &FN_SourceLocation(Position:Vector3f=&self:AttributeReader_RenderPosition(Default on Invalid001:Vector3f=<0 0 0>))
   &InitializeRibbon(
    Color=&MultiplyLinearColor(Linear Color=$User.EmissiveColor, Scale Factor=<1 1 1 0.5>)
    Module.Write Color=true
    Module.Write Lifetime=true
    Module.Write Mass=true
    Module.Write Material Random=true
    Module.Write Position=false
    Module.Write RibbonTwist=false
    Module.Write RibbonWidth=true
    Ribbon Width=10
   )
   &AddVelocity(Velocity=<0 -800 800>)
  ]
  particleUpdateScript [
   &ParticleState()
   &SolveForcesAndVelocity(Clamp Velocity=false, Limit Acceleration=false)
   &FN_UpdateRibbon_v5(Locked Position:Vector3f=&self:AttributeReader_RenderPosition(Default on Invalid001:Vector3f=<0 0 0>))
   /SetVariables(Particles.RibbonWidth=5) {isSetParameters=true, module=SetVariables_#2}
   /SetVariables(Particles.RibbonFacing=&RotateVectorByQuaternion(
     Vector To Rotate=<0 1 0>
     Quaternion=&self:AttributeReader_02(Default:Quat4f=<0 0 0 1>)
    )) {isSetParameters=true, module=SetVariables_#3}
   &ScaleColor(
    Scale Alpha=&FloatFromCurve(FloatCurve=DataInterface() {propertyValues {Curve {keys [<at 0 0>,<at 0.2 1>]}}}, CurveIndex=$User.ParaGlideLeanAlpha)
    ScaleA=true
    ScaleRGB=true
   )
  ]
  rendererClasses [&NiagaraRibbonRendererProperties]
  renderers [{rendererClass=&NiagaraRibbonRendererProperties}]
 }
 Frontend_Ribbon_RHand {
  emitterSpawnScript [&FN_Source_v2(AttributeReader=DataInterface() {propertyValues {EmitterBinding {bindingMode=Other, emitterName=RHand_Frontend}}})]
  emitterUpdateScript [
   &EmitterState(
    Delay First Loop Only=true
    Inactive Response=::NewEnumerator1
    Life Cycle Mode=::NewEnumerator1
    Loop Delay=1
    Loop Duration=1
    UseLoopDelay=true
   )
   &SpawnBurst_Instantaneous(
    Spawn Count=4
    Spawn Group=-1
    Spawn Probability=&MakeCustomFloatFromBool(Bool=$User.bIsFrontEndPreview, True Float=1, False Float=0)
    Use Spawn Probability=true
   )
   &SpawnRate(
    Spawn Probability=&FN_FrontEndFloatModifier(Gameplay=0, Front End=1)
    SpawnRate=&FN_GetEmitterParticleCount(AttributeReader:NiagaraDataInterfaceParticleRead=$Emitter.AttributeReader, Multiplier:NiagaraFloat=4)
    Use Spawn Probability=true
   )
   &FN_EmitterDivingAttributes_v1(Accumulate Dive=-0.0025, Accumulate Rest=0, SkeletalMesh:NiagaraDataInterfaceSkeletalMesh=$User.SkeletalMesh)
  ]
  particleSpawnScript [
   /SetVariables(Particles.Locked:NiagaraBool=&SetBoolByIntComparison(
     A=&ReturnParticleIDAttributes(Particle ID Attribute=::NewEnumerator1)
     B=3
     Comparison Type=::NewEnumerator7
    )) {isSetParameters=true, module=SetVariables_#1}
   &FN_SourceParentID(Ribbon ID from Parent ID=true)
   &FN_SourceLocation(Position:Vector3f=&self:AttributeReader_RenderPosition(Default on Invalid001:Vector3f=<0 0 0>))
   &InitializeRibbon(
    Color=&MultiplyLinearColor(Linear Color=$User.EmissiveColor, Scale Factor=<1 1 1 0.5>)
    Module.Write Color=true
    Module.Write Lifetime=true
    Module.Write Mass=true
    Module.Write Material Random=true
    Module.Write Position=false
    Module.Write RibbonTwist=false
    Module.Write RibbonWidth=true
    Ribbon Width=10
   )
   &AddVelocity(Velocity=<0 -800 800>)
  ]
  particleUpdateScript [
   &ParticleState()
   &SolveForcesAndVelocity(Clamp Velocity=false, Limit Acceleration=false)
   &FN_UpdateRibbon_v5(Locked Position:Vector3f=&self:AttributeReader_RenderPosition(Default on Invalid001:Vector3f=<0 0 0>))
   /SetVariables(Particles.RibbonWidth=5) {isSetParameters=true, module=SetVariables_#2}
   /SetVariables(Particles.RibbonFacing=&RotateVectorByQuaternion(
     Vector To Rotate=<0 1 0>
     Quaternion=&self:AttributeReader_02(Default:Quat4f=<0 0 0 1>)
    )) {isSetParameters=true, module=SetVariables_#3}
   &ScaleColor(
    Scale Alpha=&FloatFromCurve(FloatCurve=DataInterface() {propertyValues {Curve {keys [<at 0 0>,<at 0.2 1>]}}}, CurveIndex=$User.ParaGlideLeanAlpha)
    ScaleA=true
    ScaleRGB=true
   )
  ]
  rendererClasses [&NiagaraRibbonRendererProperties]
  renderers [{rendererClass=&NiagaraRibbonRendererProperties}]
 }
 LFoot {
  emitterUpdateScript [
   &EmitterState(
    Inactive Response=::NewEnumerator1
    Life Cycle Mode=::NewEnumerator1
    Loop Behavior=::NewEnumerator1
    Loop Duration Mode=::NewEnumerator1
    UseLoopDelay=false
   )
   /SetVariables(Emitter.SpawnCount=1) {isSetParameters=true, module=SetVariables_#4}
   &SpawnBurst_Instantaneous(
    Spawn Count=&MakeCustomIntFromBool(Bool=$User.bIsFrontEndPreview, True INT=0, False INT=$Emitter.SpawnCount)
    Use Spawn Probability=false
   )
   &FN_EmitterDivingAttributes_v1(Accumulate Dive=-0.0025, Accumulate Rest=0, SkeletalMesh:NiagaraDataInterfaceSkeletalMesh=$User.SkeletalMesh)
  ]
  particleSpawnScript [
   &InitializeParticle(
    Color=$User.EmissiveColor
    Lifetime=5
    Mesh Scale=<2.5 2.5 2.5>
    Module.Write Color=true
    Module.Write Lifetime=true
    Module.Write Mass=true
    Module.Write Material Random=true
    Module.Write Position=true
    Module.Write Scale=true
    Module.Write SpriteRotation=false
    Module.Write SpriteSize=false
    Sprite Size=<5 5>
   )
   &FN_SetSpawnGroup_v2(Number of Spawn Groups=4) {enabled=false, module=FN_SetSpawnGroup_v2001}
   &FN_SetSpawnGroup(Number of Spawn Groups=1)
   /SetVariables(Particles.NormalizedSpawnGroup:NiagaraFloat=&Divide_Float(A=&MakeFloatFromInt(Int=&SpawnGroup()), B=4)) {isSetParameters=true, module=SetVariables_#5}
  ]
  particleUpdateScript [
   &ParticleState(Kill Particles When Lifetime Has Elapsed=false, Let Infinitely Lived Particles Die When Emitter Deactivates=true)
   &SkeletalMeshLocation(
    Bone / Socket Index=2
    Bone Sampling Mode=::NewEnumerator4
    Inherit Velocity=false
    Interpolate Mesh Sampling=false
    MaskSpawnGroup=false
    Sampled Position Offset=<-40 0 0>
    Sampled Position Offset Coordinate Space:ENiagaraMeshTransforms=::NewEnumerator3
    Skeletal Mesh=$User.SkeletalMesh
   ) {module=SkeletalMeshLocation001}
   &self:Flower_03(
    Angle=0.9304
    Open:NiagaraFloat=&FloatFromCurve(
     FloatCurve=DataInterface() {propertyValues {Curve {keys [<at 0 8 RCIM_Cubic>, {arriveTangent=-9.3453, interpMode=RCIM_Cubic, leaveTangent=-9.3453, tangentMode=RCTM_User, time=1, value=3.5}]}}}
     CurveIndex=$Particles.Age
     Scale Curve=$Emitter.Dive
    )
    SpawnCount:NiagaraInt32=$Emitter.SpawnCount
   ) {module=Flower_03}
   /SetVariables(Particles.MeshOrientation=&MultiplyQuaternion(
     Quaternion A=$Output.SkeletalMeshLocation001.SampledRotation
     Quaternion B=&MultiplyQuaternion(
      Quaternion A=&MakeQuatFromAxisAngle(Axis=<0 0 1>, AngleInDegrees=180)
      Quaternion B=$Particles.MeshOrientation
     )
    )) {isSetParameters=true, module=SetVariables_#6}
   &SolveForcesAndVelocity(Clamp Velocity=false, Limit Acceleration=false)
   &ScaleMeshSize(Scale Factor=&VectorFromFloat(Value=&FloatFromCurve(FloatCurve=DataInterface() {propertyValues {Curve {keys [<at 0 0>,<at 1 1>]}}}, CurveIndex=$Particles.Age, Scale Curve=$Emitter.Dive)))
   /SetVariables(Particles.Position=&Add_Vector(
     A=Linked() {linkedVariable {name=Particles.Position, type=NiagaraPosition}}
     B=&MakeVector(
      X=&Lerp_Float(
       A=&Sine(Normalized Angle=$Emitter.Age, Period=1.5, Scale=-1, Bias=0)
       B=&Sine(Normalized Angle=$Emitter.Age, Period=3, Scale=1, Bias=0)
       Alpha=&Sine(
        Normalized Angle=$Emitter.Age
        Period=2.548
        Scale=1
        Bias=&Multiply_Float(A=1, B=$Particles.NormalizedSpawnGroup)
       )
      )
      Y=&Lerp_Float(
       A=&Sine(Normalized Angle=$Emitter.Age, Period=0.6588, Scale=1, Bias=0)
       B=&Sine(Normalized Angle=$Emitter.Age, Period=3, Scale=1, Bias=0)
       Alpha=&Sine(
        Normalized Angle=$Emitter.Age
        Period=2.45
        Scale=1
        Bias=&Multiply_Float(A=1, B=$Particles.NormalizedSpawnGroup)
       )
      )
      Z=&Add_Float(
       A=0
       B=&Lerp_Float(
        A=&Sine(Normalized Angle=$Emitter.Age, Period=3, Scale=1, Bias=0)
        B=&Sine(Normalized Angle=$Emitter.Age, Period=1.7574, Scale=1, Bias=0)
        Alpha=&Sine(
         Normalized Angle=$Emitter.Age
         Period=1.5
         Scale=1
         Bias=&Multiply_Float(A=1, B=$Particles.NormalizedSpawnGroup)
        )
       )
      )
     )
    )) {isSetParameters=true, module=SetVariables_#7}
   /SetVariables(Particles.RenderPosition:Vector3f=&Add_Vector(
     A=&RotateVectorByQuaternion(Vector To Rotate=<15 0 0>, Quaternion=$Particles.MeshOrientation)
     B=Linked() {linkedVariable {name=Particles.Position, type=NiagaraPosition}}
    )) {isSetParameters=true, module=SetVariables_#8}
   &ScaleColor(
    Scale RGB=&VectorFromFloat(Value=&Lerp_Float(A=0, B=1, Alpha=$Emitter.Dive))
    ScaleA=true
    ScaleRGB=true
   )
  ]
  rendererClasses [&NiagaraMeshRendererProperties]
  renderers [{rendererClass=&NiagaraMeshRendererProperties, rendererIndex=0}]
 }
 LFoot_Frontend {
  emitterUpdateScript [
   &EmitterState(
    Inactive Response=::NewEnumerator1
    Life Cycle Mode=::NewEnumerator1
    Loop Behavior=::NewEnumerator1
    Loop Duration Mode=::NewEnumerator1
    UseLoopDelay=false
   )
   /SetVariables(Emitter.SpawnCount=1) {isSetParameters=true, module=SetVariables_#4}
   &SpawnBurst_Instantaneous(
    Spawn Count=&MakeCustomIntFromBool(Bool=$User.bIsFrontEndPreview, True INT=$Emitter.SpawnCount, False INT=0)
    Use Spawn Probability=false
   )
   &FN_EmitterDivingAttributes_v1(Accumulate Dive=-0.0025, Accumulate Rest=0, SkeletalMesh:NiagaraDataInterfaceSkeletalMesh=$User.SkeletalMesh)
  ]
  particleSpawnScript [
   &InitializeParticle(
    Color=$User.EmissiveColor
    Lifetime=5
    Mesh Scale=<2.5 2.5 2.5>
    Module.Write Color=true
    Module.Write Lifetime=true
    Module.Write Mass=true
    Module.Write Material Random=true
    Module.Write Position=true
    Module.Write Scale=true
    Module.Write SpriteRotation=false
    Module.Write SpriteSize=false
    Sprite Size=<5 5>
   ) {enabled=false}
   /Niagara/Modules/Spawn/Initialization/V2/InitializeParticle(
    Color=$User.EmissiveColor
    Lifetime=5
    Mass Mode=::NewEnumerator1
    Mesh Scale=<2.5 2.5 2.5>
    Mesh Scale Mode=::NewEnumerator4
    Sprite Size Mode=::NewEnumerator4
    UsePositionOffset=false
    Write Material Random=true
   ) {module=InitializeParticle001}
   &FN_SetSpawnGroup_v2(Number of Spawn Groups=1) {enabled=false, module=FN_SetSpawnGroup_v2001}
   &FN_SetSpawnGroup(Number of Spawn Groups=1)
   /SetVariables(Particles.NormalizedSpawnGroup:NiagaraFloat=&Divide_Float(A=&MakeFloatFromInt(Int=&SpawnGroup()), B=4)) {isSetParameters=true, module=SetVariables_#5}
  ]
  particleUpdateScript [
   &ParticleState(Kill Particles When Lifetime Has Elapsed=false, Let Infinitely Lived Particles Die When Emitter Deactivates=true)
   &SkeletalMeshLocation(
    Bone / Socket Index=2
    Bone Sampling Mode=::NewEnumerator4
    Inherit Velocity=false
    Interpolate Mesh Sampling=false
    MaskSpawnGroup=false
    Sampled Position Offset=<-40 0 0>
    Sampled Position Offset Coordinate Space:ENiagaraMeshTransforms=::NewEnumerator3
    Skeletal Mesh=$User.SkeletalMesh
   ) {module=SkeletalMeshLocation001}
   &self:Flower_03(
    Angle=0.25
    Open:NiagaraFloat=&FloatFromCurve(
     FloatCurve=DataInterface() {propertyValues {Curve {keys [{arriveTangent=-4.3022, interpMode=RCIM_Cubic, leaveTangent=-4.3022, tangentMode=RCTM_User, time=1, value=5}]}}}
     CurveIndex=$Particles.Age
     Scale Curve=&FloatFromCurve(FloatCurve=DataInterface() {propertyValues {Curve {keys [<at 0 1>,<at 0.2 1>]}}}, CurveIndex=$User.ParaGlideLeanAlpha)
    )
    SpawnCount:NiagaraInt32=$Emitter.SpawnCount
   ) {module=Flower_03}
   /SetVariables(Particles.MeshOrientation=&MultiplyQuaternion(
     Quaternion A=$Output.SkeletalMeshLocation001.SampledRotation
     Quaternion B=&MultiplyQuaternion(
      Quaternion A=&MakeQuatFromAxisAngle(Axis=<0 0 1>, AngleInDegrees=0)
      Quaternion B=$Particles.MeshOrientation
     )
    )) {isSetParameters=true, module=SetVariables_#6}
   &SolveForcesAndVelocity(Clamp Velocity=false, Limit Acceleration=false)
   &ScaleMeshSize(Scale Factor=&VectorFromFloat(Value=&FloatFromCurve(
      FloatCurve=DataInterface() {propertyValues {Curve {keys [<at 0 0>,<at 1 1>]}}}
      CurveIndex=$Particles.Age
      Scale Curve=&FloatFromCurve(FloatCurve=DataInterface() {propertyValues {Curve {keys [<at 0 0>,<at 0.2 1>]}}}, CurveIndex=$User.ParaGlideLeanAlpha)
     )))
   /SetVariables(Particles.Position=&Add_Vector(
     A=Linked() {linkedVariable {name=Particles.Position, type=NiagaraPosition}}
     B=&MakeVector(
      X=&Lerp_Float(
       A=&Sine(Normalized Angle=$Emitter.Age, Period=1.5, Scale=-1, Bias=0)
       B=&Sine(Normalized Angle=$Emitter.Age, Period=3, Scale=1, Bias=0)
       Alpha=&Sine(
        Normalized Angle=$Emitter.Age
        Period=2.548
        Scale=1
        Bias=&Multiply_Float(A=1, B=$Particles.NormalizedSpawnGroup)
       )
      )
      Y=&Lerp_Float(
       A=&Sine(Normalized Angle=$Emitter.Age, Period=0.6588, Scale=1, Bias=0)
       B=&Sine(Normalized Angle=$Emitter.Age, Period=3, Scale=1, Bias=0)
       Alpha=&Sine(
        Normalized Angle=$Emitter.Age
        Period=2.45
        Scale=1
        Bias=&Multiply_Float(A=1, B=$Particles.NormalizedSpawnGroup)
       )
      )
      Z=&Add_Float(
       A=0
       B=&Lerp_Float(
        A=&Sine(Normalized Angle=$Emitter.Age, Period=3, Scale=1, Bias=0)
        B=&Sine(Normalized Angle=$Emitter.Age, Period=1.7574, Scale=1, Bias=0)
        Alpha=&Sine(
         Normalized Angle=$Emitter.Age
         Period=1.5
         Scale=1
         Bias=&Multiply_Float(A=1, B=$Particles.NormalizedSpawnGroup)
        )
       )
      )
     )
    )) {isSetParameters=true, module=SetVariables_#7}
   /SetVariables(Particles.RenderPosition:Vector3f=&Add_Vector(
     A=&RotateVectorByQuaternion(Vector To Rotate=<15 0 0>, Quaternion=$Particles.MeshOrientation)
     B=Linked() {linkedVariable {name=Particles.Position, type=NiagaraPosition}}
    )) {isSetParameters=true, module=SetVariables_#8}
   &ScaleColor(
    Scale RGB=&VectorFromFloat(Value=&Lerp_Float(A=0, B=1, Alpha=$User.ParaGlideLeanAlpha))
    ScaleA=true
    ScaleRGB=true
   )
  ]
  rendererClasses [&NiagaraMeshRendererProperties]
  renderers [{rendererClass=&NiagaraMeshRendererProperties, rendererIndex=0}]
 }
 LHand {
  emitterUpdateScript [
   &EmitterState(
    Inactive Response=::NewEnumerator1
    Life Cycle Mode=::NewEnumerator1
    Loop Behavior=::NewEnumerator1
    Loop Duration Mode=::NewEnumerator1
    UseLoopDelay=false
   )
   /SetVariables(Emitter.SpawnCount=4) {isSetParameters=true, module=SetVariables_#4}
   &SpawnBurst_Instantaneous(
    Spawn Count=&MakeCustomIntFromBool(Bool=$User.bIsFrontEndPreview, True INT=0, False INT=$Emitter.SpawnCount)
    Use Spawn Probability=false
   )
   &FN_EmitterDivingAttributes_v1(Accumulate Dive=-0.0025, Accumulate Rest=0, SkeletalMesh:NiagaraDataInterfaceSkeletalMesh=$User.SkeletalMesh)
  ]
  particleSpawnScript [
   /Niagara/Modules/Spawn/Initialization/InitializeParticle(
    Color=$User.EmissiveColor
    Lifetime=5
    Mesh Scale=<2.5 2.5 2.5>
    Module.Write Color=true
    Module.Write Lifetime=true
    Module.Write Mass=true
    Module.Write Material Random=true
    Module.Write Position=true
    Module.Write Scale=true
    Module.Write SpriteRotation=false
    Module.Write SpriteSize=false
    Sprite Size=<5 5>
   )
   &FN_SetSpawnGroup_v2(Number of Spawn Groups=4) {module=FN_SetSpawnGroup_v2001}
   /SetVariables(Particles.NormalizedSpawnGroup:NiagaraFloat=&Divide_Float(A=&MakeFloatFromInt(Int=&SpawnGroup()), B=4)) {isSetParameters=true, module=SetVariables_#5}
  ]
  particleUpdateScript [
   &ParticleState(Kill Particles When Lifetime Has Elapsed=false, Let Infinitely Lived Particles Die When Emitter Deactivates=true)
   &SkeletalMeshLocation(
    Bone Sampling Mode=::NewEnumerator4
    Inherit Velocity=false
    Interpolate Mesh Sampling=false
    MaskSpawnGroup=false
    Sampled Position Offset=<20 0 0>
    Sampled Position Offset Coordinate Space:ENiagaraMeshTransforms=::NewEnumerator3
    Skeletal Mesh=$User.SkeletalMesh
   ) {module=SkeletalMeshLocation001}
   &self:Flower_01(
    Angle=0.9304
    Open:NiagaraFloat=&FloatFromCurve(
     FloatCurve=DataInterface() {propertyValues {Curve {keys [<at 0 8 RCIM_Cubic>, {arriveTangent=-11.8763, interpMode=RCIM_Cubic, leaveTangent=-11.8763, tangentMode=RCTM_User, time=1, value=3.5}]}}}
     CurveIndex=$Particles.Age
     Scale Curve=$Emitter.Dive
    )
    SpawnCount:NiagaraInt32=$Emitter.SpawnCount
   ) {module=Flower_01}
   /SetVariables(Particles.MeshOrientation=&MultiplyQuaternion(Quaternion A=$Output.SkeletalMeshLocation001.SampledRotation, Quaternion B=$Particles.MeshOrientation)) {isSetParameters=true, module=SetVariables_#6}
   &SolveForcesAndVelocity(Clamp Velocity=false, Limit Acceleration=false)
   &ScaleMeshSize(Scale Factor=&VectorFromFloat(Value=&FloatFromCurve(
      FloatCurve=DataInterface() {propertyValues {Curve {keys [<at 0 0>,<at 1 1>]}}}
      CurveIndex=$Particles.Age
      Scale Curve=&FloatFromCurve(FloatCurve=DataInterface() {propertyValues {Curve {keys [<at 0 0.5>,<at 1 1>]}}}, CurveIndex=$Emitter.Dive)
     )))
   /SetVariables(Particles.Position=&Add_Vector(
     A=Linked() {linkedVariable {name=Particles.Position, type=NiagaraPosition}}
     B=&MakeVector(
      X=&Lerp_Float(
       A=&Sine(Normalized Angle=$Emitter.Age, Period=1.5, Scale=-1, Bias=0)
       B=&Sine(Normalized Angle=$Emitter.Age, Period=3, Scale=1, Bias=0)
       Alpha=&Sine(
        Normalized Angle=$Emitter.Age
        Period=2.548
        Scale=1
        Bias=&Multiply_Float(A=1, B=$Particles.NormalizedSpawnGroup)
       )
      )
      Y=&Lerp_Float(
       A=&Sine(Normalized Angle=$Emitter.Age, Period=0.6588, Scale=1, Bias=0)
       B=&Sine(Normalized Angle=$Emitter.Age, Period=3, Scale=1, Bias=0)
       Alpha=&Sine(
        Normalized Angle=$Emitter.Age
        Period=2.45
        Scale=1
        Bias=&Multiply_Float(A=1, B=$Particles.NormalizedSpawnGroup)
       )
      )
      Z=&Add_Float(
       A=0
       B=&Lerp_Float(
        A=&Sine(Normalized Angle=$Emitter.Age, Period=3, Scale=1, Bias=0)
        B=&Sine(Normalized Angle=$Emitter.Age, Period=1.7574, Scale=1, Bias=0)
        Alpha=&Sine(
         Normalized Angle=$Emitter.Age
         Period=1.5
         Scale=1
         Bias=&Multiply_Float(A=1, B=$Particles.NormalizedSpawnGroup)
        )
       )
      )
     )
    )) {isSetParameters=true, module=SetVariables_#9}
   /SetVariables(Particles.RenderPosition:Vector3f=&Add_Vector(
     A=&RotateVectorByQuaternion(
      Vector To Rotate=&VectorFromCurve(
       VectorCurve=DataInterface() {propertyValues {XCurve {keys [<at 0 0>,<at 1 15>]}, YCurve {keys [<at 0 0>,<at 1 0>]}, ZCurve {keys [<at 0 10>,<at 1 0>]}}}
       CurveIndex=$Emitter.Dive
      )
      Quaternion=$Particles.MeshOrientation
     )
     B=Linked() {linkedVariable {name=Particles.Position, type=NiagaraPosition}}
    )) {isSetParameters=true, module=SetVariables_#8}
   &ScaleColor(
    Scale RGB=&VectorFromFloat(Value=&Lerp_Float(A=0, B=1, Alpha=$Emitter.Dive))
    ScaleA=true
    ScaleRGB=true
   )
  ]
  rendererClasses [&NiagaraMeshRendererProperties]
  renderers [{rendererClass=&NiagaraMeshRendererProperties, rendererIndex=0}]
 }
 LHand_Frontend {
  emitterUpdateScript [
   &EmitterState(
    Inactive Response=::NewEnumerator1
    Life Cycle Mode=::NewEnumerator1
    Loop Behavior=::NewEnumerator1
    Loop Duration Mode=::NewEnumerator1
    UseLoopDelay=false
   )
   /SetVariables(Emitter.SpawnCount=4) {isSetParameters=true, module=SetVariables_#4}
   &SpawnBurst_Instantaneous(
    Spawn Count=&MakeCustomIntFromBool(Bool=$User.bIsFrontEndPreview, True INT=$Emitter.SpawnCount, False INT=0)
    Use Spawn Probability=false
   )
   &FN_EmitterDivingAttributes_v1(Accumulate Dive=-0.0025, Accumulate Rest=0, SkeletalMesh:NiagaraDataInterfaceSkeletalMesh=$User.SkeletalMesh)
  ]
  particleSpawnScript [
   &InitializeParticle(
    Color=$User.EmissiveColor
    Lifetime=5
    Mesh Scale=<2.5 2.5 2.5>
    Module.Write Color=true
    Module.Write Lifetime=true
    Module.Write Mass=true
    Module.Write Material Random=true
    Module.Write Position=true
    Module.Write Scale=true
    Module.Write SpriteRotation=false
    Module.Write SpriteSize=false
    Sprite Size=<5 5>
   )
   &FN_SetSpawnGroup_v2(Number of Spawn Groups=4) {module=FN_SetSpawnGroup_v2001}
   /SetVariables(Particles.NormalizedSpawnGroup:NiagaraFloat=&Divide_Float(A=&MakeFloatFromInt(Int=&SpawnGroup()), B=4)) {isSetParameters=true, module=SetVariables_#5}
  ]
  particleUpdateScript [
   &ParticleState(Kill Particles When Lifetime Has Elapsed=false, Let Infinitely Lived Particles Die When Emitter Deactivates=true)
   &SkeletalMeshLocation(
    Bone Sampling Mode=::NewEnumerator4
    Inherit Velocity=false
    Interpolate Mesh Sampling=false
    MaskSpawnGroup=false
    Sampled Position Offset=<20 0 0>
    Sampled Position Offset Coordinate Space:ENiagaraMeshTransforms=::NewEnumerator3
    Skeletal Mesh=$User.SkeletalMesh
   ) {module=SkeletalMeshLocation001}
   &self:Flower_01(
    Angle=0.75
    Open:NiagaraFloat=&FloatFromCurve(
     FloatCurve=DataInterface() {propertyValues {Curve {keys [{arriveTangent=-4.3022, interpMode=RCIM_Cubic, leaveTangent=-4.3022, tangentMode=RCTM_User, time=1, value=5}]}}}
     CurveIndex=$Particles.Age
     Scale Curve=&FloatFromCurve(FloatCurve=DataInterface() {propertyValues {Curve {keys [<at 0 0>,<at 0.2 1>]}}}, CurveIndex=$User.ParaGlideLeanAlpha)
    )
    SpawnCount:NiagaraInt32=$Emitter.SpawnCount
   ) {module=Flower_01}
   /SetVariables(Particles.MeshOrientation=&MultiplyQuaternion(Quaternion A=$Output.SkeletalMeshLocation001.SampledRotation, Quaternion B=$Particles.MeshOrientation)) {isSetParameters=true, module=SetVariables_#6}
   &SolveForcesAndVelocity(Clamp Velocity=false, Limit Acceleration=false)
   &ScaleMeshSize(Scale Factor=&VectorFromFloat(Value=&FloatFromCurve(
      FloatCurve=DataInterface() {propertyValues {Curve {keys [<at 0 0.5>,<at 1 1>]}}}
      CurveIndex=$Particles.Age
      Scale Curve=&FloatFromCurve(FloatCurve=DataInterface() {propertyValues {Curve {keys [<at 0 0.5>,<at 0.2 1>]}}}, CurveIndex=$User.ParaGlideLeanAlpha)
     )))
   /SetVariables(Particles.Position=&Add_Vector(
     A=Linked() {linkedVariable {name=Particles.Position, type=NiagaraPosition}}
     B=&MakeVector(
      X=&Lerp_Float(
       A=&Sine(Normalized Angle=$Emitter.Age, Period=1.5, Scale=-1, Bias=0)
       B=&Sine(Normalized Angle=$Emitter.Age, Period=3, Scale=1, Bias=0)
       Alpha=&Sine(
        Normalized Angle=$Emitter.Age
        Period=2.548
        Scale=1
        Bias=&Multiply_Float(A=1, B=$Particles.NormalizedSpawnGroup)
       )
      )
      Y=&Lerp_Float(
       A=&Sine(Normalized Angle=$Emitter.Age, Period=0.6588, Scale=1, Bias=0)
       B=&Sine(Normalized Angle=$Emitter.Age, Period=3, Scale=1, Bias=0)
       Alpha=&Sine(
        Normalized Angle=$Emitter.Age
        Period=2.45
        Scale=1
        Bias=&Multiply_Float(A=1, B=$Particles.NormalizedSpawnGroup)
       )
      )
      Z=&Add_Float(
       A=0
       B=&Lerp_Float(
        A=&Sine(Normalized Angle=$Emitter.Age, Period=3, Scale=1, Bias=0)
        B=&Sine(Normalized Angle=$Emitter.Age, Period=1.7574, Scale=1, Bias=0)
        Alpha=&Sine(
         Normalized Angle=$Emitter.Age
         Period=1.5
         Scale=1
         Bias=&Multiply_Float(A=1, B=$Particles.NormalizedSpawnGroup)
        )
       )
      )
     )
    )) {isSetParameters=true, module=SetVariables_#9}
   /SetVariables(Particles.RenderPosition:Vector3f=&Add_Vector(
     A=&RotateVectorByQuaternion(
      Vector To Rotate=&VectorFromCurve(
       VectorCurve=DataInterface() {propertyValues {XCurve {keys [<at 0 0>,<at 1 15>]}, YCurve {keys [<at 0 0>,<at 1 0>]}, ZCurve {keys [<at 0 10>,<at 1 0>]}}}
       CurveIndex=&FloatFromCurve(FloatCurve=DataInterface() {propertyValues {Curve {keys [<at 0 0>,<at 0.2 1>]}}}, CurveIndex=$User.ParaGlideLeanAlpha)
      )
      Quaternion=$Particles.MeshOrientation
     )
     B=Linked() {linkedVariable {name=Particles.Position, type=NiagaraPosition}}
    )) {isSetParameters=true, module=SetVariables_#8}
   &ScaleColor(
    Scale RGB=&VectorFromFloat(Value=&Lerp_Float(A=0, B=1, Alpha=$User.ParaGlideLeanAlpha))
    ScaleA=true
    ScaleRGB=true
   )
  ]
  rendererClasses [&NiagaraMeshRendererProperties]
  renderers [{rendererClass=&NiagaraMeshRendererProperties, rendererIndex=0}]
 }
 RFoot {
  emitterUpdateScript [
   &EmitterState(
    Inactive Response=::NewEnumerator1
    Life Cycle Mode=::NewEnumerator1
    Loop Behavior=::NewEnumerator1
    Loop Duration Mode=::NewEnumerator1
    UseLoopDelay=false
   )
   /SetVariables(Emitter.SpawnCount=1) {isSetParameters=true, module=SetVariables_#4}
   &SpawnBurst_Instantaneous(
    Spawn Count=&MakeCustomIntFromBool(Bool=$User.bIsFrontEndPreview, True INT=0, False INT=$Emitter.SpawnCount)
    Use Spawn Probability=false
   )
   &FN_EmitterDivingAttributes_v1(Accumulate Dive=-0.0025, Accumulate Rest=0, SkeletalMesh:NiagaraDataInterfaceSkeletalMesh=$User.SkeletalMesh)
  ]
  particleSpawnScript [
   &InitializeParticle(
    Color=$User.EmissiveColor
    Lifetime=5
    Mesh Scale=<2.5 2.5 2.5>
    Module.Write Color=true
    Module.Write Lifetime=true
    Module.Write Mass=true
    Module.Write Material Random=true
    Module.Write Position=true
    Module.Write Scale=true
    Module.Write SpriteRotation=false
    Module.Write SpriteSize=false
    Sprite Size=<5 5>
   )
   &FN_SetSpawnGroup_v2(Number of Spawn Groups=4) {enabled=false, module=FN_SetSpawnGroup_v2001}
   &FN_SetSpawnGroup(Number of Spawn Groups=1)
   /SetVariables(Particles.NormalizedSpawnGroup:NiagaraFloat=&Divide_Float(A=&MakeFloatFromInt(Int=&SpawnGroup()), B=4)) {isSetParameters=true, module=SetVariables_#5}
  ]
  particleUpdateScript [
   &ParticleState(Kill Particles When Lifetime Has Elapsed=false, Let Infinitely Lived Particles Die When Emitter Deactivates=true)
   &SkeletalMeshLocation(
    Bone / Socket Index=3
    Bone Sampling Mode=::NewEnumerator4
    Inherit Velocity=false
    Interpolate Mesh Sampling=false
    MaskSpawnGroup=false
    Sampled Position Offset=<40 0 0>
    Sampled Position Offset Coordinate Space:ENiagaraMeshTransforms=::NewEnumerator3
    Skeletal Mesh=$User.SkeletalMesh
   ) {module=SkeletalMeshLocation001}
   &self:Flower_04(
    Angle=0.9304
    Open:NiagaraFloat=&FloatFromCurve(
     FloatCurve=DataInterface() {propertyValues {Curve {keys [<at 0 8 RCIM_Cubic>, {arriveTangent=-9.5122, interpMode=RCIM_Cubic, leaveTangent=-9.5122, tangentMode=RCTM_User, time=1, value=3.5}]}}}
     CurveIndex=$Particles.Age
     Scale Curve=$Emitter.Dive
    )
    SpawnCount:NiagaraInt32=$Emitter.SpawnCount
   ) {module=Flower_04}
   /SetVariables(Particles.MeshOrientation=&MultiplyQuaternion(Quaternion A=$Output.SkeletalMeshLocation001.SampledRotation, Quaternion B=$Particles.MeshOrientation)) {isSetParameters=true, module=SetVariables_#6}
   &SolveForcesAndVelocity(Clamp Velocity=false, Limit Acceleration=false)
   &ScaleMeshSize(Scale Factor=&VectorFromFloat(Value=&FloatFromCurve(FloatCurve=DataInterface() {propertyValues {Curve {keys [<at 0 0>,<at 1 1>]}}}, CurveIndex=$Particles.Age, Scale Curve=$Emitter.Dive)))
   /SetVariables(Particles.Position=&Add_Vector(
     A=Linked() {linkedVariable {name=Particles.Position, type=NiagaraPosition}}
     B=&MakeVector(
      X=&Lerp_Float(
       A=&Sine(Normalized Angle=$Emitter.Age, Period=1.5, Scale=-1, Bias=0)
       B=&Sine(Normalized Angle=$Emitter.Age, Period=3, Scale=1, Bias=0)
       Alpha=&Sine(
        Normalized Angle=$Emitter.Age
        Period=2.548
        Scale=1
        Bias=&Multiply_Float(A=1, B=$Particles.NormalizedSpawnGroup)
       )
      )
      Y=&Lerp_Float(
       A=&Sine(Normalized Angle=$Emitter.Age, Period=0.6588, Scale=1, Bias=0)
       B=&Sine(Normalized Angle=$Emitter.Age, Period=3, Scale=1, Bias=0)
       Alpha=&Sine(
        Normalized Angle=$Emitter.Age
        Period=2.45
        Scale=1
        Bias=&Multiply_Float(A=1, B=$Particles.NormalizedSpawnGroup)
       )
      )
      Z=&Add_Float(
       A=0
       B=&Lerp_Float(
        A=&Sine(Normalized Angle=$Emitter.Age, Period=3, Scale=1, Bias=0)
        B=&Sine(Normalized Angle=$Emitter.Age, Period=1.7574, Scale=1, Bias=0)
        Alpha=&Sine(
         Normalized Angle=$Emitter.Age
         Period=1.5
         Scale=1
         Bias=&Multiply_Float(A=1, B=$Particles.NormalizedSpawnGroup)
        )
       )
      )
     )
    )) {isSetParameters=true, module=SetVariables_#10}
   /SetVariables(Particles.RenderPosition:Vector3f=&Add_Vector(
     A=&RotateVectorByQuaternion(Vector To Rotate=<15 0 0>, Quaternion=$Particles.MeshOrientation)
     B=Linked() {linkedVariable {name=Particles.Position, type=NiagaraPosition}}
    )) {isSetParameters=true, module=SetVariables_#8}
   &ScaleColor(
    Scale RGB=&VectorFromFloat(Value=&Lerp_Float(A=0, B=1, Alpha=$Emitter.Dive))
    ScaleA=true
    ScaleRGB=true
   )
  ]
  rendererClasses [&NiagaraMeshRendererProperties]
  renderers [{rendererClass=&NiagaraMeshRendererProperties, rendererIndex=0}]
 }
 RFoot_Frontend {
  emitterUpdateScript [
   &EmitterState(
    Inactive Response=::NewEnumerator1
    Life Cycle Mode=::NewEnumerator1
    Loop Behavior=::NewEnumerator1
    Loop Duration Mode=::NewEnumerator1
    UseLoopDelay=false
   )
   /SetVariables(Emitter.SpawnCount=1) {isSetParameters=true, module=SetVariables_#4}
   &SpawnBurst_Instantaneous(
    Spawn Count=&MakeCustomIntFromBool(Bool=$User.bIsFrontEndPreview, True INT=$Emitter.SpawnCount, False INT=0)
    Use Spawn Probability=false
   )
   &FN_EmitterDivingAttributes_v1(Accumulate Dive=-0.0025, Accumulate Rest=0, SkeletalMesh:NiagaraDataInterfaceSkeletalMesh=$User.SkeletalMesh)
  ]
  particleSpawnScript [
   &InitializeParticle(
    Color=$User.EmissiveColor
    Lifetime=5
    Mesh Scale=<2.5 2.5 2.5>
    Module.Write Color=true
    Module.Write Lifetime=true
    Module.Write Mass=true
    Module.Write Material Random=true
    Module.Write Position=true
    Module.Write Scale=true
    Module.Write SpriteRotation=false
    Module.Write SpriteSize=false
    Sprite Size=<5 5>
   ) {enabled=false}
   /Niagara/Modules/Spawn/Initialization/V2/InitializeParticle(
    Color=$User.EmissiveColor
    Lifetime=5
    Mass Mode=::NewEnumerator1
    Mesh Scale=<2.5 2.5 2.5>
    Mesh Scale Mode=::NewEnumerator4
    Sprite Size Mode=::NewEnumerator4
    UsePositionOffset=false
    Write Material Random=true
   ) {module=InitializeParticle001}
   &FN_SetSpawnGroup_v2(Number of Spawn Groups=1) {enabled=false, module=FN_SetSpawnGroup_v2001}
   &FN_SetSpawnGroup(Number of Spawn Groups=1)
   /SetVariables(Particles.NormalizedSpawnGroup:NiagaraFloat=&Divide_Float(A=&MakeFloatFromInt(Int=&SpawnGroup()), B=4)) {isSetParameters=true, module=SetVariables_#5}
  ]
  particleUpdateScript [
   &ParticleState(Kill Particles When Lifetime Has Elapsed=false, Let Infinitely Lived Particles Die When Emitter Deactivates=true)
   &SkeletalMeshLocation(
    Bone / Socket Index=3
    Bone Sampling Mode=::NewEnumerator4
    Inherit Velocity=false
    Interpolate Mesh Sampling=false
    MaskSpawnGroup=false
    Sampled Position Offset=<40 0 0>
    Sampled Position Offset Coordinate Space:ENiagaraMeshTransforms=::NewEnumerator3
    Skeletal Mesh=$User.SkeletalMesh
   ) {module=SkeletalMeshLocation001}
   &self:Flower_04(
    Angle=0.75
    Open:NiagaraFloat=&FloatFromCurve(
     FloatCurve=DataInterface() {propertyValues {Curve {keys [{arriveTangent=-4.3022, interpMode=RCIM_Cubic, leaveTangent=-4.3022, tangentMode=RCTM_User, time=1, value=5}]}}}
     CurveIndex=$Particles.Age
     Scale Curve=&FloatFromCurve(FloatCurve=DataInterface() {propertyValues {Curve {keys [<at 0 1>,<at 0.2 1>]}}}, CurveIndex=$User.ParaGlideLeanAlpha)
    )
    SpawnCount:NiagaraInt32=$Emitter.SpawnCount
   ) {module=Flower_04}
   /SetVariables(Particles.MeshOrientation=&MultiplyQuaternion(Quaternion A=$Output.SkeletalMeshLocation001.SampledRotation, Quaternion B=$Particles.MeshOrientation)) {isSetParameters=true, module=SetVariables_#6}
   &SolveForcesAndVelocity(Clamp Velocity=false, Limit Acceleration=false)
   &ScaleMeshSize(Scale Factor=&VectorFromFloat(Value=&FloatFromCurve(
      FloatCurve=DataInterface() {propertyValues {Curve {keys [<at 0 0>,<at 1 1>]}}}
      CurveIndex=$Particles.Age
      Scale Curve=&FloatFromCurve(FloatCurve=DataInterface() {propertyValues {Curve {keys [<at 0 0>,<at 0.2 1>]}}}, CurveIndex=$User.ParaGlideLeanAlpha)
     )))
   /SetVariables(Particles.Position=&Add_Vector(
     A=Linked() {linkedVariable {name=Particles.Position, type=NiagaraPosition}}
     B=&MakeVector(
      X=&Lerp_Float(
       A=&Sine(Normalized Angle=$Emitter.Age, Period=1.5, Scale=-1, Bias=0)
       B=&Sine(Normalized Angle=$Emitter.Age, Period=3, Scale=1, Bias=0)
       Alpha=&Sine(
        Normalized Angle=$Emitter.Age
        Period=2.548
        Scale=1
        Bias=&Multiply_Float(A=1, B=$Particles.NormalizedSpawnGroup)
       )
      )
      Y=&Lerp_Float(
       A=&Sine(Normalized Angle=$Emitter.Age, Period=0.6588, Scale=1, Bias=0)
       B=&Sine(Normalized Angle=$Emitter.Age, Period=3, Scale=1, Bias=0)
       Alpha=&Sine(
        Normalized Angle=$Emitter.Age
        Period=2.45
        Scale=1
        Bias=&Multiply_Float(A=1, B=$Particles.NormalizedSpawnGroup)
       )
      )
      Z=&Add_Float(
       A=0
       B=&Lerp_Float(
        A=&Sine(Normalized Angle=$Emitter.Age, Period=3, Scale=1, Bias=0)
        B=&Sine(Normalized Angle=$Emitter.Age, Period=1.7574, Scale=1, Bias=0)
        Alpha=&Sine(
         Normalized Angle=$Emitter.Age
         Period=1.5
         Scale=1
         Bias=&Multiply_Float(A=1, B=$Particles.NormalizedSpawnGroup)
        )
       )
      )
     )
    )) {isSetParameters=true, module=SetVariables_#10}
   /SetVariables(Particles.RenderPosition:Vector3f=&Add_Vector(
     A=&RotateVectorByQuaternion(Vector To Rotate=<15 0 0>, Quaternion=$Particles.MeshOrientation)
     B=Linked() {linkedVariable {name=Particles.Position, type=NiagaraPosition}}
    )) {isSetParameters=true, module=SetVariables_#8}
   &ScaleColor(
    Scale RGB=&VectorFromFloat(Value=&Lerp_Float(A=0, B=1, Alpha=$User.ParaGlideLeanAlpha))
    ScaleA=true
    ScaleRGB=true
   )
  ]
  rendererClasses [&NiagaraMeshRendererProperties]
  renderers [{rendererClass=&NiagaraMeshRendererProperties, rendererIndex=0}]
 }
 RHand {
  emitterUpdateScript [
   &EmitterState(
    Inactive Response=::NewEnumerator1
    Life Cycle Mode=::NewEnumerator1
    Loop Behavior=::NewEnumerator1
    Loop Duration Mode=::NewEnumerator1
    UseLoopDelay=false
   )
   /SetVariables(Emitter.SpawnCount=4) {isSetParameters=true, module=SetVariables_#4}
   &SpawnBurst_Instantaneous(
    Spawn Count=&MakeCustomIntFromBool(Bool=$User.bIsFrontEndPreview, True INT=0, False INT=$Emitter.SpawnCount)
    Use Spawn Probability=false
   )
   &FN_EmitterDivingAttributes_v1(Accumulate Dive=-0.0025, Accumulate Rest=0, SkeletalMesh:NiagaraDataInterfaceSkeletalMesh=$User.SkeletalMesh)
  ]
  particleSpawnScript [
   /Niagara/Modules/Spawn/Initialization/InitializeParticle(
    Color=$User.EmissiveColor
    Lifetime=5
    Mesh Scale=<2.5 2.5 2.5>
    Module.Write Color=true
    Module.Write Lifetime=true
    Module.Write Mass=true
    Module.Write Material Random=true
    Module.Write Position=true
    Module.Write Scale=true
    Module.Write SpriteRotation=false
    Module.Write SpriteSize=false
    Sprite Size=<5 5>
   )
   &FN_SetSpawnGroup_v2(Number of Spawn Groups=4) {module=FN_SetSpawnGroup_v2001}
   /SetVariables(Particles.NormalizedSpawnGroup:NiagaraFloat=&Divide_Float(A=&MakeFloatFromInt(Int=&SpawnGroup()), B=4)) {isSetParameters=true, module=SetVariables_#5}
  ]
  particleUpdateScript [
   &ParticleState(Kill Particles When Lifetime Has Elapsed=false, Let Infinitely Lived Particles Die When Emitter Deactivates=true)
   &SkeletalMeshLocation(
    Bone / Socket Index=1
    Bone Sampling Mode=::NewEnumerator4
    Inherit Velocity=false
    Interpolate Mesh Sampling=false
    MaskSpawnGroup=false
    Sampled Position Offset=<-20 0 0>
    Sampled Position Offset Coordinate Space:ENiagaraMeshTransforms=::NewEnumerator3
    Skeletal Mesh=$User.SkeletalMesh
   ) {module=SkeletalMeshLocation001}
   &self:Flower_02(
    Angle=0.9304
    Open:NiagaraFloat=&FloatFromCurve(
     FloatCurve=DataInterface() {propertyValues {Curve {keys [<at 0 8 RCIM_Cubic>, {arriveTangent=-12.7159, interpMode=RCIM_Cubic, leaveTangent=-12.7159, tangentMode=RCTM_User, time=1, value=3.5}]}}}
     CurveIndex=$Particles.Age
     Scale Curve=$Emitter.Dive
    )
    SpawnCount:NiagaraInt32=$Emitter.SpawnCount
   ) {module=Flower_02}
   /SetVariables(Particles.MeshOrientation=&MultiplyQuaternion(
     Quaternion A=$Output.SkeletalMeshLocation001.SampledRotation
     Quaternion B=&MultiplyQuaternion(
      Quaternion A=&MakeQuatFromAxisAngle(Axis=<0 0 1>, AngleInDegrees=180)
      Quaternion B=$Particles.MeshOrientation
     )
    )) {isSetParameters=true, module=SetVariables_#6}
   &SolveForcesAndVelocity(Clamp Velocity=false, Limit Acceleration=false)
   &ScaleMeshSize(Scale Factor=&VectorFromFloat(Value=&FloatFromCurve(
      FloatCurve=DataInterface() {propertyValues {Curve {keys [<at 0 0>,<at 1 1>]}}}
      CurveIndex=$Particles.Age
      Scale Curve=&FloatFromCurve(FloatCurve=DataInterface() {propertyValues {Curve {keys [<at 0 0.5>,<at 1 1>]}}}, CurveIndex=$Emitter.Dive)
     )))
   /SetVariables(Particles.Position=&Add_Vector(
     A=Linked() {linkedVariable {name=Particles.Position, type=NiagaraPosition}}
     B=&MakeVector(
      X=&Lerp_Float(
       A=&Sine(Normalized Angle=$Emitter.Age, Period=1.5, Scale=-1, Bias=0)
       B=&Sine(Normalized Angle=$Emitter.Age, Period=3, Scale=1, Bias=0)
       Alpha=&Sine(
        Normalized Angle=$Emitter.Age
        Period=2.548
        Scale=1
        Bias=&Multiply_Float(A=1, B=$Particles.NormalizedSpawnGroup)
       )
      )
      Y=&Lerp_Float(
       A=&Sine(Normalized Angle=$Emitter.Age, Period=0.6588, Scale=1, Bias=0)
       B=&Sine(Normalized Angle=$Emitter.Age, Period=3, Scale=1, Bias=0)
       Alpha=&Sine(
        Normalized Angle=$Emitter.Age
        Period=2.45
        Scale=1
        Bias=&Multiply_Float(A=1, B=$Particles.NormalizedSpawnGroup)
       )
      )
      Z=&Add_Float(
       A=0
       B=&Lerp_Float(
        A=&Sine(Normalized Angle=$Emitter.Age, Period=3, Scale=1, Bias=0)
        B=&Sine(Normalized Angle=$Emitter.Age, Period=1.7574, Scale=1, Bias=0)
        Alpha=&Sine(
         Normalized Angle=$Emitter.Age
         Period=1.5
         Scale=1
         Bias=&Multiply_Float(A=1, B=$Particles.NormalizedSpawnGroup)
        )
       )
      )
     )
    )) {isSetParameters=true, module=SetVariables_#11}
   /SetVariables(Particles.RenderPosition:Vector3f=&Add_Vector(
     A=&RotateVectorByQuaternion(
      Vector To Rotate=&VectorFromCurve(
       VectorCurve=DataInterface() {propertyValues {XCurve {keys [<at 0 0>,<at 1 15>]}, YCurve {keys [<at 0 0>,<at 1 0>]}, ZCurve {keys [<at 0 10>,<at 1 0>]}}}
       CurveIndex=$Emitter.Dive
      )
      Quaternion=$Particles.MeshOrientation
     )
     B=Linked() {linkedVariable {name=Particles.Position, type=NiagaraPosition}}
    )) {isSetParameters=true, module=SetVariables_#8}
   &ScaleColor(
    Scale RGB=&VectorFromFloat(Value=&Lerp_Float(A=0, B=1, Alpha=$Emitter.Dive))
    ScaleA=true
    ScaleRGB=true
   )
  ]
  rendererClasses [&NiagaraMeshRendererProperties]
  renderers [{rendererClass=&NiagaraMeshRendererProperties, rendererIndex=0}]
 }
 RHand_Frontend {
  emitterUpdateScript [
   &EmitterState(
    Inactive Response=::NewEnumerator1
    Life Cycle Mode=::NewEnumerator1
    Loop Behavior=::NewEnumerator1
    Loop Duration Mode=::NewEnumerator1
    UseLoopDelay=false
   )
   /SetVariables(Emitter.SpawnCount=4) {isSetParameters=true, module=SetVariables_#4}
   &SpawnBurst_Instantaneous(
    Spawn Count=&MakeCustomIntFromBool(Bool=$User.bIsFrontEndPreview, True INT=$Emitter.SpawnCount, False INT=0)
    Use Spawn Probability=false
   )
   &FN_EmitterDivingAttributes_v1(Accumulate Dive=-0.0025, Accumulate Rest=0, SkeletalMesh:NiagaraDataInterfaceSkeletalMesh=$User.SkeletalMesh)
  ]
  particleSpawnScript [
   &InitializeParticle(
    Color=$User.EmissiveColor
    Lifetime=5
    Mesh Scale=<2.5 2.5 2.5>
    Module.Write Color=true
    Module.Write Lifetime=true
    Module.Write Mass=true
    Module.Write Material Random=true
    Module.Write Position=true
    Module.Write Scale=true
    Module.Write SpriteRotation=false
    Module.Write SpriteSize=false
    Sprite Size=<5 5>
   )
   &FN_SetSpawnGroup_v2(Number of Spawn Groups=4) {module=FN_SetSpawnGroup_v2001}
   /SetVariables(Particles.NormalizedSpawnGroup:NiagaraFloat=&Divide_Float(A=&MakeFloatFromInt(Int=&SpawnGroup()), B=4)) {isSetParameters=true, module=SetVariables_#5}
  ]
  particleUpdateScript [
   &ParticleState(Kill Particles When Lifetime Has Elapsed=false, Let Infinitely Lived Particles Die When Emitter Deactivates=true)
   &SkeletalMeshLocation(
    Bone / Socket Index=1
    Bone Sampling Mode=::NewEnumerator4
    Inherit Velocity=false
    Interpolate Mesh Sampling=false
    MaskSpawnGroup=false
    Sampled Position Offset=<-20 0 0>
    Sampled Position Offset Coordinate Space:ENiagaraMeshTransforms=::NewEnumerator3
    Skeletal Mesh=$User.SkeletalMesh
   ) {module=SkeletalMeshLocation001}
   &self:Flower_02(
    Angle=0.75
    Open:NiagaraFloat=&FloatFromCurve(
     FloatCurve=DataInterface() {propertyValues {Curve {keys [{arriveTangent=-4.3022, interpMode=RCIM_Cubic, leaveTangent=-4.3022, tangentMode=RCTM_User, time=1, value=5}]}}}
     CurveIndex=$Particles.Age
     Scale Curve=&FloatFromCurve(FloatCurve=DataInterface() {propertyValues {Curve {keys [<at 0 0>,<at 0.2 1>]}}}, CurveIndex=$User.ParaGlideLeanAlpha)
    )
    SpawnCount:NiagaraInt32=$Emitter.SpawnCount
   ) {module=Flower_02}
   /SetVariables(Particles.MeshOrientation=&MultiplyQuaternion(
     Quaternion A=$Output.SkeletalMeshLocation001.SampledRotation
     Quaternion B=&MultiplyQuaternion(
      Quaternion A=&MakeQuatFromAxisAngle(Axis=<0 0 1>, AngleInDegrees=180)
      Quaternion B=$Particles.MeshOrientation
     )
    )) {isSetParameters=true, module=SetVariables_#6}
   &SolveForcesAndVelocity(Clamp Velocity=false, Limit Acceleration=false)
   &ScaleMeshSize(Scale Factor=&VectorFromFloat(Value=&FloatFromCurve(
      FloatCurve=DataInterface() {propertyValues {Curve {keys [<at 0 0.5>,<at 1 1>]}}}
      CurveIndex=$Particles.Age
      Scale Curve=&FloatFromCurve(FloatCurve=DataInterface() {propertyValues {Curve {keys [<at 0 0.5>,<at 0.2 1>]}}}, CurveIndex=$User.ParaGlideLeanAlpha)
     )))
   /SetVariables(Particles.Position=&Add_Vector(
     A=Linked() {linkedVariable {name=Particles.Position, type=NiagaraPosition}}
     B=&MakeVector(
      X=&Lerp_Float(
       A=&Sine(Normalized Angle=$Emitter.Age, Period=1.5, Scale=-1, Bias=0)
       B=&Sine(Normalized Angle=$Emitter.Age, Period=3, Scale=1, Bias=0)
       Alpha=&Sine(
        Normalized Angle=$Emitter.Age
        Period=2.548
        Scale=1
        Bias=&Multiply_Float(A=1, B=$Particles.NormalizedSpawnGroup)
       )
      )
      Y=&Lerp_Float(
       A=&Sine(Normalized Angle=$Emitter.Age, Period=0.6588, Scale=1, Bias=0)
       B=&Sine(Normalized Angle=$Emitter.Age, Period=3, Scale=1, Bias=0)
       Alpha=&Sine(
        Normalized Angle=$Emitter.Age
        Period=2.45
        Scale=1
        Bias=&Multiply_Float(A=1, B=$Particles.NormalizedSpawnGroup)
       )
      )
      Z=&Add_Float(
       A=0
       B=&Lerp_Float(
        A=&Sine(Normalized Angle=$Emitter.Age, Period=3, Scale=1, Bias=0)
        B=&Sine(Normalized Angle=$Emitter.Age, Period=1.7574, Scale=1, Bias=0)
        Alpha=&Sine(
         Normalized Angle=$Emitter.Age
         Period=1.5
         Scale=1
         Bias=&Multiply_Float(A=1, B=$Particles.NormalizedSpawnGroup)
        )
       )
      )
     )
    )) {isSetParameters=true, module=SetVariables_#11}
   /SetVariables(Particles.RenderPosition:Vector3f=&Add_Vector(
     A=&RotateVectorByQuaternion(
      Vector To Rotate=&VectorFromCurve(
       VectorCurve=DataInterface() {propertyValues {XCurve {keys [<at 0 0>,<at 1 15>]}, YCurve {keys [<at 0 0>,<at 1 0>]}, ZCurve {keys [<at 0 10>,<at 1 0>]}}}
       CurveIndex=&FloatFromCurve(FloatCurve=DataInterface() {propertyValues {Curve {keys [<at 0 0>,<at 0.2 1>]}}}, CurveIndex=$User.ParaGlideLeanAlpha)
      )
      Quaternion=$Particles.MeshOrientation
     )
     B=Linked() {linkedVariable {name=Particles.Position, type=NiagaraPosition}}
    )) {isSetParameters=true, module=SetVariables_#8}
   &ScaleColor(
    Scale RGB=&VectorFromFloat(Value=&Lerp_Float(A=0, B=1, Alpha=$User.ParaGlideLeanAlpha))
    ScaleA=true
    ScaleRGB=true
   )
  ]
  rendererClasses [&NiagaraMeshRendererProperties]
  renderers [{rendererClass=&NiagaraMeshRendererProperties, rendererIndex=0}]
 }
 Ribbon_LFoot {
  emitterUpdateScript [
   &FN_Source_v2(AttributeReader=DataInterface() {propertyValues {EmitterBinding {bindingMode=Other, emitterName=LFoot}}})
   &EmitterState(
    Inactive Response=::NewEnumerator1
    Life Cycle Mode=::NewEnumerator1
    Loop Behavior=::NewEnumerator1
    Loop Duration Mode=::NewEnumerator1
    UseLoopDelay=false
   )
   &FN_EmitterDivingAttributes_v1(Accumulate Dive=-0.0025, Accumulate Rest=0, SkeletalMesh:NiagaraDataInterfaceSkeletalMesh=$User.SkeletalMesh)
   &SpawnRate(
    SpawnRate=&MakeCustomFloatFromBool(
     Bool=$User.bIsFrontEndPreview
     True Float=0
     False Float=&FN_GetEmitterParticleCount(AttributeReader:NiagaraDataInterfaceParticleRead=$Emitter.AttributeReader, Multiplier:NiagaraFloat=4)
    )
    Use Spawn Probability=false
   )
  ]
  particleSpawnScript [
   &FN_SourceParentID(Ribbon ID from Parent ID=true)
   &FN_SourceLocation(Position:Vector3f=&self:AttributeReader_RenderPosition_03(Default on Invalid001:Vector3f=<0 0 0>))
   /SetVariables(Particles.Locked:NiagaraBool=&SetBoolByIntComparison(
     A=&ReturnParticleIDAttributes(Particle ID Attribute=::NewEnumerator1)
     B=3
     Comparison Type=::NewEnumerator7
    )) {isSetParameters=true, module=SetVariables_#12}
   &InitializeRibbon(
    Color=$User.EmissiveColor
    Module.Write Color=true
    Module.Write Lifetime=true
    Module.Write Mass=false
    Module.Write Material Random=true
    Module.Write Position=false
    Module.Write RibbonTwist=false
    Module.Write RibbonWidth=true
   )
   &InitializeParticle(
    Module.Write Color=true
    Module.Write Lifetime=true
    Module.Write Mass=true
    Module.Write Material Random=true
    Module.Write Position=false
    Module.Write Scale=false
    Module.Write SpriteRotation=false
    Module.Write SpriteSize=false
   )
  ]
  particleUpdateScript [
   &ParticleState()
   &FN_UpdateRibbon_v5(
    Initial Color:LinearColor=$Particles.Initial.Color
    Locked Position:Vector3f=&self:AttributeReader_RenderPosition_03(Default on Invalid001:Vector3f=<0 0 0>)
   )
   &Color(
    Color=&MultiplyLinearColor(
     Linear Color=$User.EmissiveColor
     Scale Factor=&Vector4FromCurve(Vector4Curve=DataInterface() {propertyValues {
        AlphaCurve {keys [<at 0 0.5>,<at 1 0>]}
        BlueCurve {keys [<at 0 1>,<at 1 1>]}
        GreenCurve {keys [<at 0 1>,<at 1 1>]}
        RedCurve {keys [<at 0 1>,<at 1 1>]}
       }})
    )
    Scale Alpha=&FloatFromCurve(FloatCurve=DataInterface() {propertyValues {Curve {keys [<at 0 0>,<at 1 1>]}}}, CurveIndex=$Emitter.Dive)
   )
   &SolveForcesAndVelocity(Clamp Velocity=false, Limit Acceleration=false)
   /SetVariables(Particles.RibbonFacing=&RotateVectorByQuaternion(
     Vector To Rotate=<0 1 0>
     Quaternion=&self:AttributeReader_05(Default:Quat4f=<0 0 0 1>)
    )) {isSetParameters=true, module=SetVariables_#13}
  ]
  rendererClasses [&NiagaraRibbonRendererProperties]
  renderers [{rendererClass=&NiagaraRibbonRendererProperties}]
 }
 Ribbon_LFoot001 {
  bEnabled=false
  emitterUpdateScript [
   &FN_Source_v2(AttributeReader=DataInterface() {propertyValues {EmitterBinding {bindingMode=Other, emitterName=LFoot}}})
   &EmitterState()
   &FN_EmitterDivingAttributes_v1(Accumulate Dive=-0.0025, Accumulate Rest=0, SkeletalMesh:NiagaraDataInterfaceSkeletalMesh=$User.SkeletalMesh)
   &SpawnRate(
    SpawnRate=&FN_GetEmitterParticleCount(AttributeReader:NiagaraDataInterfaceParticleRead=$Emitter.AttributeReader, Multiplier:NiagaraFloat=80)
    Use Spawn Probability=false
   )
  ]
  particleSpawnScript [
   &FN_SourceParentID()
   &FN_SourceLocation(Position:Vector3f=&self:AttributeReader_RenderPosition(Default on Invalid001:Vector3f=<0 0 0>))
   &InitializeRibbon(
    Color=&MultiplyLinearColor(
     Linear Color=&MultiplyLinearColor(
      Linear Color=$User.EmissiveColor
      Scale Factor=&MakeVector4(
       X=1
       Y=1
       Z=1
       W=&FloatFromCurve(FloatCurve=DataInterface() {propertyValues {Curve {keys [<at 0 0>,<at 0.5 1>]}}}, CurveIndex=$Emitter.Dive)
      )
     )
     Scale Factor=<1 1 1 0.05>
    )
    Lifetime=0.08
    Module.Write Color=true
    Module.Write Lifetime=true
    Module.Write Mass=true
    Module.Write Material Random=true
    Module.Write Position=false
    Module.Write RibbonTwist=false
    Module.Write RibbonWidth=true
   )
  ]
  particleUpdateScript [
   &ParticleState()
   /SetVariables(
    Particles.Ribbon2_Color:LinearColor=&HueShiftLinearColor(
     Hue=0.95
     Saturation=1
     Value=1
     Linear Color=&MultiplyLinearColor(
      Linear Color=$User.EmissiveColor
      Scale Factor=&MakeVector4(
       X=10
       Y=10
       Z=10
       W=&FloatFromCurve(
        FloatCurve=DataInterface() {propertyValues {Curve {keys [{arriveTangent=-1.541, interpMode=RCIM_Cubic, time=0, value=0}, <at 0.25 0.4 RCIM_Cubic>, <at 0.5 0 RCIM_Cubic>]}}}
        Scale Curve=&FloatFromCurve(FloatCurve=DataInterface() {propertyValues {Curve {keys [<at 0 0>,<at 0.5 1>]}}}, CurveIndex=$Emitter.Dive)
       )
      )
     )
    )
    Particles.Ribbon2_Ribbonwidth=2.5
   ) {isSetParameters=true, module=SetVariables_#14}
   /NiagaraFN/DynamicInputs/Contrails/UpdateLockedParticleSpawnGroup()
   &InheritVelocity(Module.Limit Speed=true) {enabled=false}
   &SolveForcesAndVelocity(Clamp Velocity=false, Limit Acceleration=false)
   &ScaleColor(
    Scale Alpha=&FloatFromCurve(FloatCurve=DataInterface() {propertyValues {Curve {keys [<at 0 1>, {interpMode=RCIM_Cubic, leaveTangent=-1.6667, time=0.4, value=1}, <at 0.5 0 RCIM_Cubic>]}}})
    ScaleA=true
    ScaleRGB=true
   ) {enabled=false}
  ]
  rendererClasses [&NiagaraRibbonRendererProperties]
  renderers [
   {rendererClass=&NiagaraRibbonRendererProperties}
   {rendererClass=&NiagaraRibbonRendererProperties, rendererIndex=1}
  ]
 }
 Ribbon_LHand001 {
  bEnabled=false
  emitterUpdateScript [
   &FN_Source_v2(AttributeReader=DataInterface() {propertyValues {EmitterBinding {bindingMode=Other, emitterName=LHand}}})
   &EmitterState()
   &FN_EmitterDivingAttributes_v1(Accumulate Dive=-0.0025, Accumulate Rest=0, SkeletalMesh:NiagaraDataInterfaceSkeletalMesh=$User.SkeletalMesh)
   &SpawnRate(
    SpawnRate=&FN_GetEmitterParticleCount(AttributeReader:NiagaraDataInterfaceParticleRead=$Emitter.AttributeReader, Multiplier:NiagaraFloat=20)
    Use Spawn Probability=false
   )
  ]
  particleSpawnScript [
   &FN_SourceParentID()
   &FN_SourceLocation(Position:Vector3f=&self:AttributeReader_RenderPosition(Default on Invalid001:Vector3f=<0 0 0>))
   &InitializeRibbon(
    Color=&MultiplyLinearColor(
     Linear Color=&MultiplyLinearColor(
      Linear Color=$User.EmissiveColor
      Scale Factor=&MakeVector4(
       X=1
       Y=1
       Z=1
       W=&FloatFromCurve(FloatCurve=DataInterface() {propertyValues {Curve {keys [<at 0 0>,<at 0.5 1>]}}}, CurveIndex=$Emitter.Dive)
      )
     )
     Scale Factor=<1 1 1 0.05>
    )
    Lifetime=0.08
    Module.Write Color=true
    Module.Write Lifetime=true
    Module.Write Mass=true
    Module.Write Material Random=true
    Module.Write Position=false
    Module.Write RibbonTwist=false
    Module.Write RibbonWidth=true
   )
  ]
  particleUpdateScript [
   &ParticleState()
   /SetVariables(
    Particles.Ribbon2_Color:LinearColor=&HueShiftLinearColor(
     Hue=0.95
     Saturation=1
     Value=1
     Linear Color=&MultiplyLinearColor(
      Linear Color=$User.EmissiveColor
      Scale Factor=&MakeVector4(
       X=5
       Y=5
       Z=5
       W=&Multiply_Float(
        A=&FloatFromCurve(FloatCurve=DataInterface() {propertyValues {Curve {keys [{arriveTangent=-1.541, interpMode=RCIM_Cubic, time=0, value=0}, <at 0.25 0.4 RCIM_Cubic>, <at 0.5 0 RCIM_Cubic>]}}})
        B=&FloatFromCurve(FloatCurve=DataInterface() {propertyValues {Curve {keys [<at 0 0>,<at 0.5 1>]}}}, CurveIndex=$Emitter.Dive)
       )
      )
     )
    )
    Particles.Ribbon2_Ribbonwidth=2.5
   ) {isSetParameters=true, module=SetVariables_#14}
   &UpdateLockedParticleSpawnGroup()
   &InheritVelocity(Module.Limit Speed=true) {enabled=false}
   &SolveForcesAndVelocity(Clamp Velocity=false, Limit Acceleration=false)
   &ScaleColor(
    Scale Alpha=&FloatFromCurve(FloatCurve=DataInterface() {propertyValues {Curve {keys [<at 0 1>, {interpMode=RCIM_Cubic, leaveTangent=-1.6667, time=0.5, value=1}, <at 1 0 RCIM_Cubic>]}}})
    ScaleA=true
    ScaleRGB=true
   ) {enabled=false}
  ]
  rendererClasses [&NiagaraRibbonRendererProperties]
  renderers [
   {rendererClass=&NiagaraRibbonRendererProperties}
   {rendererClass=&NiagaraRibbonRendererProperties, rendererIndex=1}
  ]
 }
 Ribbon_LHand002 {
  emitterUpdateScript [
   &FN_Source_v2(AttributeReader=DataInterface() {propertyValues {EmitterBinding {bindingMode=Other, emitterName=LHand}}})
   &EmitterState(
    Inactive Response=::NewEnumerator1
    Life Cycle Mode=::NewEnumerator1
    Loop Behavior=::NewEnumerator1
    Loop Duration Mode=::NewEnumerator1
    UseLoopDelay=false
   )
   &FN_EmitterDivingAttributes_v1(Accumulate Dive=-0.0025, Accumulate Rest=0, SkeletalMesh:NiagaraDataInterfaceSkeletalMesh=$User.SkeletalMesh)
   &SpawnRate(
    SpawnRate=&MakeCustomFloatFromBool(
     Bool=$User.bIsFrontEndPreview
     True Float=0
     False Float=&FN_GetEmitterParticleCount(AttributeReader:NiagaraDataInterfaceParticleRead=$Emitter.AttributeReader, Multiplier:NiagaraFloat=4)
    )
    Use Spawn Probability=false
   )
  ]
  particleSpawnScript [
   &FN_SourceParentID(Ribbon ID from Parent ID=true)
   &FN_SourceLocation(Position:Vector3f=&self:AttributeReader_RenderPosition_01(Default on Invalid001:Vector3f=<0 0 0>))
   /SetVariables(Particles.Locked:NiagaraBool=&SetBoolByIntComparison(
     A=&ReturnParticleIDAttributes(Particle ID Attribute=::NewEnumerator1)
     B=3
     Comparison Type=::NewEnumerator7
    )) {isSetParameters=true, module=SetVariables_#12}
   &InitializeRibbon(
    Color=$User.EmissiveColor
    Module.Write Color=true
    Module.Write Lifetime=true
    Module.Write Mass=false
    Module.Write Material Random=true
    Module.Write Position=false
    Module.Write RibbonTwist=false
    Module.Write RibbonWidth=true
   )
   &InitializeParticle(
    Module.Write Color=true
    Module.Write Lifetime=true
    Module.Write Mass=true
    Module.Write Material Random=true
    Module.Write Position=false
    Module.Write Scale=false
    Module.Write SpriteRotation=false
    Module.Write SpriteSize=false
   )
  ]
  particleUpdateScript [
   &ParticleState()
   &FN_UpdateRibbon_v5(
    Initial Color:LinearColor=$Particles.Initial.Color
    Locked Position:Vector3f=&self:AttributeReader_RenderPosition_01(Default on Invalid001:Vector3f=<0 0 0>)
   )
   &Color(
    Color=&MultiplyLinearColor(
     Linear Color=$User.EmissiveColor
     Scale Factor=&Vector4FromCurve(Vector4Curve=DataInterface() {propertyValues {
        AlphaCurve {keys [<at 0 0.5>,<at 1 0>]}
        BlueCurve {keys [<at 0 1>,<at 1 1>]}
        GreenCurve {keys [<at 0 1>,<at 1 1>]}
        RedCurve {keys [<at 0 1>,<at 1 1>]}
       }})
    )
    Scale Alpha=&FloatFromCurve(FloatCurve=DataInterface() {propertyValues {Curve {keys [<at 0 0>,<at 1 1>]}}}, CurveIndex=$Emitter.Dive)
   )
   &SolveForcesAndVelocity(Clamp Velocity=false, Limit Acceleration=false)
   /SetVariables(Particles.RibbonFacing=&RotateVectorByQuaternion(
     Vector To Rotate=<0 1 0>
     Quaternion=&self:AttributeReader_03(Default:Quat4f=<0 0 0 1>)
    )) {isSetParameters=true, module=SetVariables_#13}
  ]
  rendererClasses [&NiagaraRibbonRendererProperties]
  renderers [{rendererClass=&NiagaraRibbonRendererProperties}]
 }
 Ribbon_RFoot {
  emitterUpdateScript [
   &FN_Source_v2(AttributeReader=DataInterface() {propertyValues {EmitterBinding {bindingMode=Other, emitterName=RFoot}}})
   &EmitterState(
    Inactive Response=::NewEnumerator1
    Life Cycle Mode=::NewEnumerator1
    Loop Behavior=::NewEnumerator1
    Loop Duration Mode=::NewEnumerator1
    UseLoopDelay=false
   )
   &FN_EmitterDivingAttributes_v1(Accumulate Dive=-0.0025, Accumulate Rest=0, SkeletalMesh:NiagaraDataInterfaceSkeletalMesh=$User.SkeletalMesh)
   &SpawnRate(
    SpawnRate=&MakeCustomFloatFromBool(
     Bool=$User.bIsFrontEndPreview
     True Float=0
     False Float=&FN_GetEmitterParticleCount(AttributeReader:NiagaraDataInterfaceParticleRead=$Emitter.AttributeReader, Multiplier:NiagaraFloat=4)
    )
    Use Spawn Probability=false
   )
  ]
  particleSpawnScript [
   &FN_SourceParentID(Ribbon ID from Parent ID=true)
   &FN_SourceLocation(Position:Vector3f=&self:AttributeReader_RenderPosition_04(Default on Invalid001:Vector3f=<0 0 0>))
   /SetVariables(Particles.Locked:NiagaraBool=&SetBoolByIntComparison(
     A=&ReturnParticleIDAttributes(Particle ID Attribute=::NewEnumerator1)
     B=3
     Comparison Type=::NewEnumerator7
    )) {isSetParameters=true, module=SetVariables_#12}
   &InitializeRibbon(
    Color=$User.EmissiveColor
    Module.Write Color=true
    Module.Write Lifetime=true
    Module.Write Mass=false
    Module.Write Material Random=true
    Module.Write Position=false
    Module.Write RibbonTwist=false
    Module.Write RibbonWidth=true
   )
   &InitializeParticle(
    Module.Write Color=true
    Module.Write Lifetime=true
    Module.Write Mass=true
    Module.Write Material Random=true
    Module.Write Position=false
    Module.Write Scale=false
    Module.Write SpriteRotation=false
    Module.Write SpriteSize=false
   )
  ]
  particleUpdateScript [
   &ParticleState()
   &FN_UpdateRibbon_v5(
    Initial Color:LinearColor=$Particles.Initial.Color
    Locked Position:Vector3f=&self:AttributeReader_RenderPosition_04(Default on Invalid001:Vector3f=<0 0 0>)
   )
   &Color(
    Color=&MultiplyLinearColor(
     Linear Color=$User.EmissiveColor
     Scale Factor=&Vector4FromCurve(Vector4Curve=DataInterface() {propertyValues {
        AlphaCurve {keys [<at 0 0.5>,<at 1 0>]}
        BlueCurve {keys [<at 0 1>,<at 1 1>]}
        GreenCurve {keys [<at 0 1>,<at 1 1>]}
        RedCurve {keys [<at 0 1>,<at 1 1>]}
       }})
    )
    Scale Alpha=&FloatFromCurve(FloatCurve=DataInterface() {propertyValues {Curve {keys [<at 0 0>,<at 1 1>]}}}, CurveIndex=$Emitter.Dive)
   )
   &SolveForcesAndVelocity(Clamp Velocity=false, Limit Acceleration=false)
   /SetVariables(Particles.RibbonFacing=&RotateVectorByQuaternion(
     Vector To Rotate=<0 1 0>
     Quaternion=&self:AttributeReader_06(Default:Quat4f=<0 0 0 1>)
    )) {isSetParameters=true, module=SetVariables_#13}
  ]
  rendererClasses [&NiagaraRibbonRendererProperties]
  renderers [{rendererClass=&NiagaraRibbonRendererProperties}]
 }
 Ribbon_RFoot001 {
  bEnabled=false
  emitterUpdateScript [
   &FN_Source_v2(AttributeReader=DataInterface() {propertyValues {EmitterBinding {bindingMode=Other, emitterName=RFoot}}})
   &EmitterState()
   &FN_EmitterDivingAttributes_v1(Accumulate Dive=-0.0025, Accumulate Rest=0, SkeletalMesh:NiagaraDataInterfaceSkeletalMesh=$User.SkeletalMesh)
   &SpawnRate(
    SpawnRate=&FN_GetEmitterParticleCount(AttributeReader:NiagaraDataInterfaceParticleRead=$Emitter.AttributeReader, Multiplier:NiagaraFloat=80)
    Use Spawn Probability=false
   )
  ]
  particleSpawnScript [
   &FN_SourceParentID()
   &FN_SourceLocation(Position:Vector3f=&self:AttributeReader_RenderPosition(Default on Invalid001:Vector3f=<0 0 0>))
   &InitializeRibbon(
    Color=&MultiplyLinearColor(
     Linear Color=&MultiplyLinearColor(
      Linear Color=$User.EmissiveColor
      Scale Factor=&MakeVector4(
       X=1
       Y=1
       Z=1
       W=&FloatFromCurve(FloatCurve=DataInterface() {propertyValues {Curve {keys [<at 0 0>,<at 0.5 1>]}}}, CurveIndex=$Emitter.Dive)
      )
     )
     Scale Factor=<1 1 1 0.05>
    )
    Lifetime=0.08
    Module.Write Color=true
    Module.Write Lifetime=true
    Module.Write Mass=true
    Module.Write Material Random=true
    Module.Write Position=false
    Module.Write RibbonTwist=false
    Module.Write RibbonWidth=true
   )
  ]
  particleUpdateScript [
   &ParticleState()
   /SetVariables(
    Particles.Ribbon2_Color:LinearColor=&HueShiftLinearColor(
     Hue=0.95
     Saturation=1
     Value=1
     Linear Color=&MultiplyLinearColor(
      Linear Color=$User.EmissiveColor
      Scale Factor=&MakeVector4(
       X=10
       Y=10
       Z=10
       W=&FloatFromCurve(
        FloatCurve=DataInterface() {propertyValues {Curve {keys [{arriveTangent=-1.541, interpMode=RCIM_Cubic, time=0, value=0}, <at 0.25 0.4 RCIM_Cubic>, <at 0.5 0 RCIM_Cubic>]}}}
        Scale Curve=&FloatFromCurve(FloatCurve=DataInterface() {propertyValues {Curve {keys [<at 0 0>,<at 0.5 1>]}}}, CurveIndex=$Emitter.Dive)
       )
      )
     )
    )
    Particles.Ribbon2_Ribbonwidth=2.5
   ) {isSetParameters=true, module=SetVariables_#14}
   &UpdateLockedParticleSpawnGroup()
   &InheritVelocity(Module.Limit Speed=true) {enabled=false}
   &SolveForcesAndVelocity(Clamp Velocity=false, Limit Acceleration=false)
   &ScaleColor(
    Scale Alpha=&FloatFromCurve(FloatCurve=DataInterface() {propertyValues {Curve {keys [<at 0 1>, {interpMode=RCIM_Cubic, leaveTangent=-1.6667, time=0.5, value=1}, <at 0.75 0 RCIM_Cubic>]}}})
    ScaleA=true
    ScaleRGB=true
   ) {enabled=false}
  ]
  rendererClasses [&NiagaraRibbonRendererProperties]
  renderers [
   {rendererClass=&NiagaraRibbonRendererProperties}
   {rendererClass=&NiagaraRibbonRendererProperties, rendererIndex=1}
  ]
 }
 Ribbon_RHand {
  emitterUpdateScript [
   &FN_Source_v2(AttributeReader=DataInterface() {propertyValues {EmitterBinding {bindingMode=Other, emitterName=RHand}}})
   &EmitterState(
    Inactive Response=::NewEnumerator1
    Life Cycle Mode=::NewEnumerator1
    Loop Behavior=::NewEnumerator1
    Loop Duration Mode=::NewEnumerator1
    UseLoopDelay=false
   )
   &FN_EmitterDivingAttributes_v1(Accumulate Dive=-0.0025, Accumulate Rest=0, SkeletalMesh:NiagaraDataInterfaceSkeletalMesh=$User.SkeletalMesh)
   &SpawnRate(
    SpawnRate=&MakeCustomFloatFromBool(
     Bool=$User.bIsFrontEndPreview
     True Float=0
     False Float=&FN_GetEmitterParticleCount(AttributeReader:NiagaraDataInterfaceParticleRead=$Emitter.AttributeReader, Multiplier:NiagaraFloat=4)
    )
    Use Spawn Probability=false
   )
  ]
  particleSpawnScript [
   &FN_SourceParentID(Ribbon ID from Parent ID=true)
   &FN_SourceLocation(Position:Vector3f=&self:AttributeReader_RenderPosition_02(Default on Invalid001:Vector3f=<0 0 0>))
   /SetVariables(Particles.Locked:NiagaraBool=&SetBoolByIntComparison(
     A=&ReturnParticleIDAttributes(Particle ID Attribute=::NewEnumerator1)
     B=3
     Comparison Type=::NewEnumerator7
    )) {isSetParameters=true, module=SetVariables_#12}
   &InitializeRibbon(
    Color=$User.EmissiveColor
    Module.Write Color=true
    Module.Write Lifetime=true
    Module.Write Mass=false
    Module.Write Material Random=true
    Module.Write Position=false
    Module.Write RibbonTwist=false
    Module.Write RibbonWidth=true
   )
   &InitializeParticle(
    Module.Write Color=true
    Module.Write Lifetime=true
    Module.Write Mass=true
    Module.Write Material Random=true
    Module.Write Position=false
    Module.Write Scale=false
    Module.Write SpriteRotation=false
    Module.Write SpriteSize=false
   )
  ]
  particleUpdateScript [
   &ParticleState()
   &FN_UpdateRibbon_v5(
    Initial Color:LinearColor=$Particles.Initial.Color
    Locked Position:Vector3f=&self:AttributeReader_RenderPosition_02(Default on Invalid001:Vector3f=<0 0 0>)
   )
   &Color(
    Color=&MultiplyLinearColor(
     Linear Color=$User.EmissiveColor
     Scale Factor=&Vector4FromCurve(Vector4Curve=DataInterface() {propertyValues {
        AlphaCurve {keys [<at 0 0.5>,<at 1 0>]}
        BlueCurve {keys [<at 0 1>,<at 1 1>]}
        GreenCurve {keys [<at 0 1>,<at 1 1>]}
        RedCurve {keys [<at 0 1>,<at 1 1>]}
       }})
    )
    Scale Alpha=&FloatFromCurve(FloatCurve=DataInterface() {propertyValues {Curve {keys [<at 0 0>,<at 1 1>]}}}, CurveIndex=$Emitter.Dive)
   )
   &SolveForcesAndVelocity(Clamp Velocity=false, Limit Acceleration=false)
   /SetVariables(Particles.RibbonFacing=&RotateVectorByQuaternion(
     Vector To Rotate=<0 1 0>
     Quaternion=&self:AttributeReader_04(Default:Quat4f=<0 0 0 1>)
    )) {isSetParameters=true, module=SetVariables_#13}
  ]
  rendererClasses [&NiagaraRibbonRendererProperties]
  renderers [{rendererClass=&NiagaraRibbonRendererProperties}]
 }
 Ribbon_RHand001 {
  bEnabled=false
  emitterUpdateScript [
   &FN_Source_v2(AttributeReader=DataInterface() {propertyValues {EmitterBinding {bindingMode=Other, emitterName=RightHand}}})
   &EmitterState()
   &FN_EmitterDivingAttributes_v1(Accumulate Dive=-0.0025, Accumulate Rest=0, SkeletalMesh:NiagaraDataInterfaceSkeletalMesh=$User.SkeletalMesh)
   &SpawnRate(
    SpawnRate=&FN_GetEmitterParticleCount(AttributeReader:NiagaraDataInterfaceParticleRead=$Emitter.AttributeReader, Multiplier:NiagaraFloat=80)
    Use Spawn Probability=false
   )
  ]
  particleSpawnScript [
   &FN_SourceParentID()
   &FN_SourceLocation(Position:Vector3f=&self:AttributeReader_RenderPosition(Default on Invalid001:Vector3f=<0 0 0>))
   &InitializeRibbon(
    Color=&MultiplyLinearColor(
     Linear Color=&MultiplyLinearColor(
      Linear Color=$User.EmissiveColor
      Scale Factor=&MakeVector4(
       X=1
       Y=1
       Z=1
       W=&FloatFromCurve(FloatCurve=DataInterface() {propertyValues {Curve {keys [<at 0 0>,<at 0.5 1>]}}}, CurveIndex=$Emitter.Dive)
      )
     )
     Scale Factor=<1 1 1 0.05>
    )
    Lifetime=0.08
    Module.Write Color=true
    Module.Write Lifetime=true
    Module.Write Mass=true
    Module.Write Material Random=true
    Module.Write Position=false
    Module.Write RibbonTwist=false
    Module.Write RibbonWidth=true
   )
  ]
  particleUpdateScript [
   &ParticleState()
   /SetVariables(
    Particles.Ribbon2_Color:LinearColor=&HueShiftLinearColor(
     Hue=0.95
     Saturation=1
     Value=1
     Linear Color=&MultiplyLinearColor(
      Linear Color=$User.EmissiveColor
      Scale Factor=&MakeVector4(
       X=10
       Y=10
       Z=10
       W=&FloatFromCurve(
        FloatCurve=DataInterface() {propertyValues {Curve {keys [{arriveTangent=-1.541, interpMode=RCIM_Cubic, time=0, value=0}, <at 0.25 0.4 RCIM_Cubic>, <at 0.5 0 RCIM_Cubic>]}}}
        Scale Curve=&FloatFromCurve(FloatCurve=DataInterface() {propertyValues {Curve {keys [<at 0 0>,<at 0.5 1>]}}}, CurveIndex=$Emitter.Dive)
       )
      )
     )
    )
    Particles.Ribbon2_Ribbonwidth=2.5
   ) {isSetParameters=true, module=SetVariables_#14}
   &UpdateLockedParticleSpawnGroup()
   &InheritVelocity(Module.Limit Speed=true) {enabled=false}
   &SolveForcesAndVelocity(Clamp Velocity=false, Limit Acceleration=false)
   &ScaleColor(
    Scale Alpha=&FloatFromCurve(FloatCurve=DataInterface() {propertyValues {Curve {keys [<at 0 1>, {interpMode=RCIM_Cubic, leaveTangent=-1.6667, time=0.5, value=1}, <at 0.75 0 RCIM_Cubic>]}}})
    ScaleA=true
    ScaleRGB=true
   ) {enabled=false}
  ]
  rendererClasses [&NiagaraRibbonRendererProperties]
  renderers [
   {rendererClass=&NiagaraRibbonRendererProperties}
   {rendererClass=&NiagaraRibbonRendererProperties, rendererIndex=1}
  ]
 }
}
summary {
 systemName=NS_Contrail_Lexa_Frontend
 emitters [
  {emitterName=FN_SourceGlows, rendererClasses [&NiagaraSpriteRendererProperties]}
  {bEnabled=false, emitterName=Ribbon_LHand001, rendererClasses [&NiagaraRibbonRendererProperties]}
  {bEnabled=false, emitterName=Ribbon_RHand001, rendererClasses [&NiagaraRibbonRendererProperties]}
  {bEnabled=false, emitterName=Ribbon_LFoot001, rendererClasses [&NiagaraRibbonRendererProperties]}
  {bEnabled=false, emitterName=Ribbon_RFoot001, rendererClasses [&NiagaraRibbonRendererProperties]}
  {emitterName=Frontend_Ribbon_LHand, rendererClasses [&NiagaraRibbonRendererProperties]}
  {emitterName=Frontend_Ribbon_RHand, rendererClasses [&NiagaraRibbonRendererProperties]}
  {emitterName=Frontend_Ribbon_LFoot, rendererClasses [&NiagaraRibbonRendererProperties]}
  {emitterName=Frontend_Ribbon_RFoot, rendererClasses [&NiagaraRibbonRendererProperties]}
  {emitterName=LHand, rendererClasses [&NiagaraMeshRendererProperties]}
  {emitterName=RHand, rendererClasses [&NiagaraMeshRendererProperties]}
  {emitterName=LFoot, rendererClasses [&NiagaraMeshRendererProperties]}
  {emitterName=RFoot, rendererClasses [&NiagaraMeshRendererProperties]}
  {emitterName=Ribbon_LHand002, rendererClasses [&NiagaraRibbonRendererProperties]}
  {emitterName=Ribbon_RHand, rendererClasses [&NiagaraRibbonRendererProperties]}
  {emitterName=Ribbon_LFoot, rendererClasses [&NiagaraRibbonRendererProperties]}
  {emitterName=Ribbon_RFoot, rendererClasses [&NiagaraRibbonRendererProperties]}
  {emitterName=LHand_Frontend, rendererClasses [&NiagaraMeshRendererProperties]}
  {emitterName=RHand_Frontend, rendererClasses [&NiagaraMeshRendererProperties]}
  {emitterName=LFoot_Frontend, rendererClasses [&NiagaraMeshRendererProperties]}
  {emitterName=RFoot_Frontend, rendererClasses [&NiagaraMeshRendererProperties]}
  {emitterName=Decal, rendererClasses [&NiagaraSpriteRendererProperties, &NiagaraLightRendererProperties]}
 ]
 userVariables [
  User.bIsFrontEnd {type=NiagaraBool}
  User.bIsFrontEndPreview {type=NiagaraBool}
  User.EmissiveColor {type=LinearColor, description="", defaultValue {struct=LinearColor, value=<rgba 1 0 0 1>}}
  User.ParaGlideLeanAlpha {type=NiagaraFloat, defaultValue=0, description=""}
  User.PlayerSpeed {type=NiagaraFloat, defaultValue=0, description=""}
  User.RibbonIDOffset {type=NiagaraInt32}
  User.SkeletalMesh {
   type=NiagaraDataInterfaceSkeletalMesh
   defaultValue {
    struct=NiagaraExt_VariableValue_DataInterface
    value {dataInterface=&self:NiagaraDataInterfaceSkeletalMesh_0}
   }
  }
  User.SkeletalMesh2 {
   type=NiagaraDataInterfaceSkeletalMesh
   defaultValue {
    struct=NiagaraExt_VariableValue_DataInterface
    value {dataInterface=&self:NiagaraDataInterfaceSkeletalMesh_1}
   }
  }
 ]
}
